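# Control Rig for the HangNine female character that drives body deformation and corrective motion for clothing and accessories.
# A sequence runs the shared DeformRig function, shoelace and jacket correctives, then necklace, charm, crouch and shirt adjustments.
# Weighted parent and aim constraints place base bones.
# Spherical pose readers, joint-rotation remaps, spine-to-calf distances, pelvis height with IsCrouching, and deformRig thigh attributes weight the corrective transform offsets.
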ControlRigBlueprint /BRCosmetics/Characters/Player/Female/Medium/Bodies/F_MED_HangNine/Meshes/F_MED_HangNine_CtrlRig
graphs [
 New Function_ContainedGraph {nodes [
   RigVMFunctionEntryNode()
   > &RigUnit_ParentConstraint(
    AdvancedSettings=(RotationOrderForFilter=XZY)
    Child=(Type=Bone,Name="base_neck_shirt_lf")
    Parents="((Item=(Type=Bone,Name=\"deform_upperarm_fr_l\"),Weight=0.0019),(Item=(Type=Bone,Name=\"upperarm_twist_01_l\"),Weight=0.0031),(Item=(Type=Bone,Name=\"neck_02\"),Weight=0.0055),(Item=(Type=Bone,Name=\"neck_01\"),Weight=0.0071),(Item=(Type=Bone,Name=\"spine_03\"),Weig ...(truncated)"
    Weight=1
   )
   > &RigUnit_ParentConstraint(
    AdvancedSettings=(RotationOrderForFilter=XZY)
    Child=(Type=Bone,Name="base_neck_shirt_rt")
    Parents="((Item=(Type=Bone,Name=\"deform_upperarm_fr_r\"),Weight=0.0019),(Item=(Type=Bone,Name=\"upperarm_twist_01_r\"),Weight=0.0032),(Item=(Type=Bone,Name=\"spine_04\"),Weight=0.0034),(Item=(Type=Bone,Name=\"neck_02\"),Weight=0.006),(Item=(Type=Bone,Name=\"neck_01\"),Weigh ...(truncated)"
    Weight=1
   )
   > &RigUnit_RotationConstraintLocalSpaceOffset(
    AdvancedSettings=(InterpolationType=Average,RotationOrderForFilter=XZY)
    Child=(Type=Bone,Name="dyn_neck_shirt_lf")
    Filter=(bX=True,bY=True,bZ=True)
    Parents=((Item=(Type=Bone,Name="clavicle_l"),Weight=1))
    Weight=1
   )
   > &RigUnit_RotationConstraintLocalSpaceOffset(
    AdvancedSettings=(InterpolationType=Average,RotationOrderForFilter=XZY)
    Child=(Type=Bone,Name="dyn_neck_shirt_rt")
    Filter=(bX=True,bY=True,bZ=True)
    Parents=((Item=(Type=Bone,Name="clavicle_r"),Weight=1))
    Weight=1
   )
   > n1 = &RigUnit_SphericalPoseReader(
    ActiveRegionScaleFactors=(PositiveWidth=1,NegativeWidth=1,PositiveHeight=1,NegativeHeight=1)
    ActiveRegionSize=0.1
    Debug=(bDrawDebug=false,bDraw2D=False,bDrawLocalAxes=False,DebugScale=25,DebugSegments=20,DebugThickness=0.25)
    DriverAxis=(X=1,Y=0,Z=0)
    DriverItem=(Type=Bone,Name="clavicle_l")
    FalloffRegionScaleFactors=(PositiveWidth=1,NegativeWidth=1,PositiveHeight=1,NegativeHeight=1)
    FalloffSize=0.3
    FlipHeightScaling=False
    FlipWidthScaling=False
    OptionalParentItem=(Type=Bone,Name="None")
    RotationOffset=(X=0,Y=-30,Z=0)
   )
   > &RigUnit_OffsetTransformForItem(
    Item=(Type=Bone,Name="dyn_neck_shirt_lf")
    OffsetTransform=(Rotation=(X=0,Y=-0.2164,Z=0,W=0.9763),Translation=(X=0,Y=0,Z=0),Scale3D=(X=1,Y=1,Z=1))
    bPropagateToChildren=True
    Weight=@n1.OutputParam
   )
   > n2 = &RigUnit_SphericalPoseReader(
    ActiveRegionScaleFactors=(PositiveWidth=1,NegativeWidth=1,PositiveHeight=1,NegativeHeight=1)
    ActiveRegionSize=0.1
    Debug=(bDrawDebug=false,bDraw2D=False,bDrawLocalAxes=False,DebugScale=25,DebugSegments=20,DebugThickness=0.25)
    DriverAxis=(X=-1,Y=0,Z=0)
    DriverItem=(Type=Bone,Name="clavicle_r")
    FalloffRegionScaleFactors=(PositiveWidth=1,NegativeWidth=1,PositiveHeight=1,NegativeHeight=1)
    FalloffSize=0.3
    FlipHeightScaling=False
    FlipWidthScaling=False
    OptionalParentItem=(Type=Bone,Name="None")
    RotationOffset=(X=0,Y=150,Z=0)
   )
   > &RigUnit_OffsetTransformForItem(
    Item=(Type=Bone,Name="dyn_neck_shirt_rt")
    OffsetTransform=(Rotation=(X=0,Y=-0.2164,Z=0,W=0.9763),Translation=(X=0,Y=0,Z=0),Scale3D=(X=1,Y=1,Z=1))
    bPropagateToChildren=True
    Weight=@n2.OutputParam
   )
   > RigVMFunctionReturnNode()
  ]}
 New Function_ContainedGraph {nodes [
   n1 = RigVMFunctionEntryNode()
   RigVMFunction_MathBoolToInteger = &RigVMFunction_MathBoolToInteger(Value=@n1.onlyTwist)
   RigVMFunction_MathQuaternionSwingTwist = &RigVMFunction_MathQuaternionSwingTwist(
    twistAxis=@n1.twistAxis
    Input=&RigUnit_GetRelativeTransformForItem(
     bChildInitial=False
     bParentInitial=False
     child=(Type=Bone,Name="None")
     parent=(Type=Bone,Name="None")
     child.Name=@n1.child
     parent.Name=@n1.parent
     parent.Type=@n1.parentType
    ).RelativeTransform.Rotation
   )
   RigVMFunction_MathQuaternionToEuler = &RigVMFunction_MathQuaternionToEuler(
    Value=RigVMDispatchNode(
     Values=((),())
     Index=@RigVMFunction_MathBoolToInteger.Result
     Values.0=@RigVMFunction_MathQuaternionSwingTwist.Swing
     Values.1=@RigVMFunction_MathQuaternionSwingTwist.Twist
    ) {template="DISPATCH_RigVMDispatch_SelectInt32(in Index,in Values,out Result)"}.Result
    RotationOrder=@n1.Rotation Order
   )
   > DISPATCH_RigVMDispatch_Print = RigVMDispatchNode(ScreenColor=(R=1,G=1,B=1,A=1), ScreenDuration=1.5, Enabled=@n1.debugRotation, Value=@RigVMFunction_MathQuaternionToEuler.Result) {template="DISPATCH_RigVMDispatch_Print(in Prefix,in Value,in Enabled,in ScreenDuration,in ScreenColor)"}
   Clamp = &RigVMFunction_MathIntClamp(Value=@n1.primAxis)
   DISPATCH_RigVMDispatch_SelectInt32 = RigVMDispatchNode(
    Values=(0,0,0)
    Index=@Clamp.Result
    Values.0=@RigVMFunction_MathQuaternionToEuler.Result.X
    Values.1=@RigVMFunction_MathQuaternionToEuler.Result.Y
    Values.2=@RigVMFunction_MathQuaternionToEuler.Result.Z
   ) {template="DISPATCH_RigVMDispatch_SelectInt32(in Index,in Values,out Result)"}
   Remap = &RigVMFunction_MathDoubleRemap(
    TargetMaximum=1
    TargetMinimum=0
    bClamp=False
    Value=@DISPATCH_RigVMDispatch_SelectInt32.Result
    SourceMaximum=@n1.max
    SourceMinimum=@n1.min
   )
   Multiply = &RigVMFunction_MathDoubleMul(A=&RigVMFunction_MathDoubleClamp(Value=@Remap.Result).Result, B=@n1.premult)
   > RigVMFunctionReturnNode(outputWeight=@Multiply.Result)
  ]}
 New Function_ContainedGraph {nodes [
   n1 = RigVMFunctionEntryNode()
   RigVMFunction_MathBoolToInteger = &RigVMFunction_MathBoolToInteger(Value=@n1.onlyTwist)
   RigVMFunction_MathQuaternionSwingTwist = &RigVMFunction_MathQuaternionSwingTwist(
    twistAxis=@n1.twistAxis
    Input=&RigUnit_GetRelativeTransformForItem(
     bChildInitial=False
     bParentInitial=False
     child=(Type=Bone,Name="None")
     parent=(Type=Bone,Name="None")
     child.Name=@n1.child
     parent.Name=@n1.parent
     parent.Type=@n1.parentType
    ).RelativeTransform.Rotation
   )
   RigVMFunction_MathQuaternionToEuler = &RigVMFunction_MathQuaternionToEuler(
    Value=RigVMDispatchNode(
     Values=((),())
     Index=@RigVMFunction_MathBoolToInteger.Result
     Values.0=@RigVMFunction_MathQuaternionSwingTwist.Swing
     Values.1=@RigVMFunction_MathQuaternionSwingTwist.Twist
    ) {template="DISPATCH_RigVMDispatch_SelectInt32(in Index,in Values,out Result)"}.Result
    RotationOrder=@n1.Rotation Order
   )
   > DISPATCH_RigVMDispatch_Print = RigVMDispatchNode(ScreenColor=(R=1,G=1,B=1,A=1), ScreenDuration=1.5, Enabled=@n1.debugRotation, Value=@RigVMFunction_MathQuaternionToEuler.Result) {template="DISPATCH_RigVMDispatch_Print(in Prefix,in Value,in Enabled,in ScreenDuration,in ScreenColor)"}
   Clamp = &RigVMFunction_MathIntClamp(Value=@n1.primAxis)
   DISPATCH_RigVMDispatch_SelectInt32 = RigVMDispatchNode(
    Values=(0,0,0)
    Index=@Clamp.Result
    Values.0=@RigVMFunction_MathQuaternionToEuler.Result.X
    Values.1=@RigVMFunction_MathQuaternionToEuler.Result.Y
    Values.2=@RigVMFunction_MathQuaternionToEuler.Result.Z
   ) {template="DISPATCH_RigVMDispatch_SelectInt32(in Index,in Values,out Result)"}
   Remap = &RigVMFunction_MathDoubleRemap(
    TargetMaximum=1
    TargetMinimum=0
    bClamp=False
    Value=@DISPATCH_RigVMDispatch_SelectInt32.Result
    SourceMaximum=@n1.max
    SourceMinimum=@n1.min
   )
   Multiply = &RigVMFunction_MathDoubleMul(A=&RigVMFunction_MathDoubleClamp(Value=@Remap.Result).Result, B=@n1.premult)
   > RigVMFunctionReturnNode(outputWeight=@Multiply.Result)
  ]}
 New Function_ContainedGraph {nodes [
   n1 = RigVMFunctionEntryNode()
   RigVMFunction_MathQuaternionToEuler = &RigVMFunction_MathQuaternionToEuler(
    RotationOrder=ZYX
    Value=&RigUnit_GetRelativeTransformForItem(
     Child=(Type=Bone,Name="hand_l")
     Parent=(Type=Bone,Name="lowerarm_l")
     bChildInitial=False
     bParentInitial=False
     Child.Name=@n1.Child
     Parent.Name=@n1.Parent
    ).RelativeTransform.Rotation
   )
   Remap = &RigVMFunction_MathDoubleRemap(
    TargetMaximum=1
    TargetMinimum=0
    Value=@RigVMFunction_MathQuaternionToEuler.Result.Z
    SourceMaximum=@n1.Source Maximum
    SourceMinimum=@n1.Source Minimum
   )
   > RigVMFunctionReturnNode(delta=@Remap.Result)
  ]}
 RigVMFunctionLibrary {nodes [
   Delta Rotation = RigVMCollapseNode(ExecuteContext=())
   DeltaRotation = RigVMCollapseNode(ExecuteContext=(), Rotation Order=ZYX, premult=1, twistAxis=(X=1,Y=0,Z=0))
   DeltaRotation_1 = RigVMCollapseNode(ExecuteContext=(), Rotation Order=ZYX, premult=1, twistAxis=(X=1,Y=0,Z=0))
   neck shirt = RigVMCollapseNode(ExecuteContext=())
  ]}
 RigVMModel {nodes [
   RigUnit_BeginExecution = &RigUnit_BeginExecution()
   RigVMFunction_Sequence = &RigVMFunction_Sequence(ExecuteContext=@RigUnit_BeginExecution.ExecutePin)
   FN_DeformRig_FNC = RigVMFunctionReferenceNode(
    DeformRig_ Curve="(deformRig_l_upperarm_up_alpha_72_#1=False,deformRig_r_upperarm_up_alpha_73_#2=False,deformRig_l_upperarm_fr_alpha_74_#3=False,deformRig_r_upperarm_fr_alpha_75_CB51B8 ...(truncated)"
    DeformRig_Bool="(upperarm_1_#4=True,elbow_3_#5=True,wrist_7_#6=True,groin_8_#7=True,glute_17_#8=True,knee_9_40C408324CA261 ...(truncated)"
    DeformRig_Debug=(upperarm_debug_18_#4=False,wrist_debug_20_#6=False,lat_debug_26_#9=False,groinGlute_debug_27_#7=False)
    DeformRig_Value="(upperarm_up_18_#10=(Rotation=(X=0.0,Y=-0.0,Z=0.0,W=1.0),Translation=(X=0.0,Y=0.0,Z=8.0),Scale3D=(X=0.70,Y=1.0,Z=1.0)),upperarm_fwd_fr_19_#11=(Rotation=(X=0.0,Y=-0.0,Z=0.0,W=1.0),Translation=(X=0.0, ...(truncated)"
    ExecuteContext=@RigVMFunction_Sequence.A
   )
   > &RigUnit_ParentConstraint(
    AdvancedSettings=(RotationOrderForFilter=XZY)
    Child=(Type=Bone,Name="base_main_shoelaces_lf")
    Parents=((Item=(Type=Bone,Name="calf_l"),Weight=0.0256),(Item=(Type=Bone,Name="ball_l"),Weight=0.0318),(Item=(Type=Bone,Name="calf_twist_02_l"),Weight=0.033),(Item=(Type=Bone,Name="foot_l"),Weight=0.3077),(Item=(Type=Bone,Name="calf_twist_01_l"),Weight=0.6019))
    Weight=1
   )
   > &RigUnit_ParentConstraint(
    AdvancedSettings=(RotationOrderForFilter=XZY)
    Child=(Type=Bone,Name="base_main_shoelaces_rt")
    Parents=((Item=(Type=Bone,Name="ball_r"),Weight=0.0175),(Item=(Type=Bone,Name="calf_r"),Weight=0.0224),(Item=(Type=Bone,Name="calf_twist_02_r"),Weight=0.0231),(Item=(Type=Bone,Name="foot_r"),Weight=0.2928),(Item=(Type=Bone,Name="calf_twist_01_r"),Weight=0.6441))
    Weight=1
   )
   > Delta Rotation_1_1 = RigVMFunctionReferenceNode(Child=foot_l, Parent=calf_l, Source Maximum=-80, Source Minimum=0)
   > &RigUnit_ModifyTransforms(
    ItemToModify="((Item=(Type=Bone,Name=\"base_shoelace_lower_03_lf\"),Transform=(Rotation=(X=0.0298,Y=-0.0819,Z=0.2578,W=0.9623),Translation=(X=0,Y=0,Z=0),Scale3D=(X=1,Y=1,Z=1))),(Item=(Type=Bone,Name=\"base_shoelace_lower_lf\"),Transform=(Rotation=(X=-0.1,Y=-0.7595,Z=-0.0839 ...(truncated)"
    WeightMinimum=0
    Weight=@"Delta Rotation_1_1".delta
   )
   > Delta Rotation_1_2 = RigVMFunctionReferenceNode(Child=foot_r, Parent=calf_r, Source Maximum=-80, Source Minimum=0)
   > &RigUnit_ModifyTransforms(
    ItemToModify="((Item=(Type=Bone,Name=\"base_shoelace_lower_rt\"),Transform=(Rotation=(X=-0.0334,Y=-0.3812,Z=-0.0805,W=0.9204),Translation=(X=0,Y=0,Z=0),Scale3D=(X=1,Y=1,Z=1))),(Item=(Type=Bone,Name=\"base_bow_shoelace_lower_rt\"),Transform=(Rotation=(X=-0.1722,Y=0.0227,Z=-0 ...(truncated)"
    WeightMinimum=0
    Weight=@"Delta Rotation_1_2".delta
   )
   > &RigUnit_AimConstraintLocalSpaceOffset(
    Child=(Type=Bone,Name="base_jacket_spine_cr")
    Filter=(bX=True,bY=True,bZ=True)
    Parents=((Item=(Type=Bone,Name="pelvis"),Weight=1))
    UpAxis=(X=0,Y=1,Z=0)
    Weight=1
    WorldUp=(Target=(X=0,Y=1,Z=0),Kind=Direction,Space=(Type=Bone,Name="base_jacket_spine_cr"))
   )
   > n1 = &RigUnit_SphericalPoseReader(
    ActiveRegionScaleFactors=(PositiveWidth=1,NegativeWidth=1,PositiveHeight=1,NegativeHeight=1)
    ActiveRegionSize=0.2
    Debug=(bDrawDebug=false,bDraw2D=False,bDrawLocalAxes=False,DebugScale=25,DebugSegments=20,DebugThickness=0.25)
    DriverAxis=(X=0,Y=1,Z=0)
    DriverItem=(Type=Bone,Name="pelvis")
    FalloffRegionScaleFactors=(PositiveWidth=1,NegativeWidth=1,PositiveHeight=1,NegativeHeight=1)
    FalloffSize=0.4
    FlipHeightScaling=False
    FlipWidthScaling=False
    OptionalParentItem=(Type=Bone,Name="spine_03")
    RotationOffset=(X=90,Y=0,Z=-90)
   )
   > &RigUnit_ModifyTransforms(
    ItemToModify="((Item=(Type=Bone,Name=\"base_jacket_front_rt\"),Transform=(Rotation=(X=0,Y=-0.0872,Z=0,W=0.9962),Translation=(X=0,Y=0,Z=0),Scale3D=(X=1,Y=1,Z=1))),(Item=(Type=Bone,Name=\"base_jacket_front_lf\"),Transform=(Rotation=(X=0,Y=-0.0872,Z=0,W=0.9962),Translation=(X= ...(truncated)"
    WeightMinimum=0
    Weight=@n1.OutputParam
   )
   > n2 = &RigUnit_SphericalPoseReader(
    ActiveRegionScaleFactors=(PositiveWidth=1,NegativeWidth=1,PositiveHeight=1,NegativeHeight=1)
    ActiveRegionSize=0.15
    Debug=(bDrawDebug=false,bDraw2D=False,bDrawLocalAxes=False,DebugScale=25,DebugSegments=20,DebugThickness=0.25)
    DriverAxis=(X=1,Y=0,Z=0)
    DriverItem=(Type=Bone,Name="upperarm_l")
    FalloffRegionScaleFactors=(PositiveWidth=1,NegativeWidth=1,PositiveHeight=1,NegativeHeight=1)
    FalloffSize=0.25
    FlipHeightScaling=False
    FlipWidthScaling=false
    OptionalParentItem=(Type=Bone,Name="None")
    RotationOffset=(X=0,Y=-45,Z=20)
   )
   > &RigUnit_ParentConstraint(
    AdvancedSettings=(InterpolationType=Average,RotationOrderForFilter=XZY)
    Child=(Type=Bone,Name="spine_03_for_cons")
    Parents=((Item=(Type=Bone,Name="spine_03"),Weight=1))
    Weight=1
   )
   n3 = &RigVMFunction_AlphaInterp(InterpSpeedDecreasing=10, InterpSpeedIncreasing=10, Scale=1, Value=1, bClampResult=False)
   n4 = &RigVMFunction_AnimEvalRichCurve(
    Curve="(EditorCurveData=(Keys=((InterpMode=RCIM_Cubic,TangentMode=RCTM_Auto,TangentWeightMode=RCTWM_WeightedNone,Time=0,Value=0,ArriveTangent=0,ArriveTangentWeight=0,LeaveTangent=0,LeaveTangentWeight=0),(InterpMode=RCIM_Cubic,TangentMode=RCTM_Auto,TangentWeightMo ...(truncated)"
    SourceMinimum=0
    TargetMinimum=0
    Value=@n3.Result
   )
   AlphaInterp = &RigVMFunction_AlphaInterp(InterpSpeedDecreasing=10, InterpSpeedIncreasing=10, Scale=1, bClampResult=false, Value=@n2.OutputParam)
   n5 = &RigVMFunction_AnimEvalRichCurve(
    Curve="(EditorCurveData=(Keys=((InterpMode=RCIM_Cubic,TangentMode=RCTM_Auto,TangentWeightMode=RCTWM_WeightedNone,Time=0,Value=0,ArriveTangent=0,ArriveTangentWeight=0,LeaveTangent=0,LeaveTangentWeight=0),(InterpMode=RCIM_Cubic,TangentMode=RCTM_Auto,TangentWeightMo ...(truncated)"
    SourceMinimum=0
    TargetMinimum=0
    Value=@AlphaInterp.Result
   )
   n6 = &RigUnit_GetTransform(Item=(Type=Bone,Name="spine_02_for_cons"), Space=LocalSpace, bInitial=false)
   RigVMFunction_MathQuaternionFromEuler = &RigVMFunction_MathQuaternionFromEuler(Euler=/Script/RigVM.RigVMFunction_MathVectorMul(
     B=(X=-1,Y=-1,Z=-1)
     B.X=&RigVMFunction_MathDoubleRemap(SourceMaximum=1, SourceMinimum=0, TargetMaximum=1, TargetMinimum=-1, Value=@n5.Result).Result
     A=&RigVMFunction_MathQuaternionToEuler(RotationOrder=ZYX, Value=@n6.Transform.Rotation).Result
    ) {is=struct RigVMUnitNode, template="Multiply::Execute(in A,in B,out Result)"}.Result)
   > Set Transform = &RigUnit_SetRotation(Item=(Type=Bone,Name="jacket_spine_cr"), bInitial=false, Weight=@n4.Result, Value=@RigVMFunction_MathQuaternionFromEuler.Result)
   RigVMFunction_MathQuaternionToEuler = &RigVMFunction_MathQuaternionToEuler(
    RotationOrder=ZYX
    Value=&RigUnit_GetRelativeTransformForItem(Child=(Type=Bone,Name="pelvis"), Parent=(Type=Bone,Name="spine_03"), bChildInitial=False, bParentInitial=False).RelativeTransform.Rotation
   )
   Remap = &RigVMFunction_MathDoubleRemap(SourceMaximum=-30, SourceMinimum=8.33, TargetMaximum=1, TargetMinimum=0, Value=@RigVMFunction_MathQuaternionToEuler.Result.Z)
   > &RigUnit_ModifyTransforms(
    ItemToModify="((Item=(Type=Bone,Name=\"base_jacket_back_cr\"),Transform=(Rotation=(X=-0.0017,Y=-0.0003,Z=0.1779,W=0.984),Translation=(X=1,Y=-8.4,Z=0),Scale3D=(X=1,Y=1,Z=1))),(Item=(Type=Bone,Name=\"base_jacket_side_back_rt\"),Transform=(Rotation=(X=0,Y=0,Z=-0.0723,W=-0.9974 ...(truncated)"
    WeightMinimum=0
    Weight=@Remap.Result
   )
   > n7 = &RigUnit_SphericalPoseReader(
    ActiveRegionScaleFactors=(PositiveWidth=1,NegativeWidth=1,PositiveHeight=1,NegativeHeight=1)
    ActiveRegionSize=0.15
    Debug=(bDrawDebug=false,bDraw2D=False,bDrawLocalAxes=False,DebugScale=25,DebugSegments=20,DebugThickness=0.25)
    DriverAxis=(X=0,Y=0,Z=1)
    DriverItem=(Type=Bone,Name="spine_03")
    FalloffRegionScaleFactors=(PositiveWidth=1,NegativeWidth=1,PositiveHeight=1,NegativeHeight=1)
    FalloffSize=0.25
    FlipHeightScaling=False
    FlipWidthScaling=false
    OptionalParentItem=(Type=Bone,Name="None")
    RotationOffset=(X=45,Y=0,Z=0)
   )
   > &RigUnit_ModifyTransforms(
    ItemToModify="((Item=(Type=Bone,Name=\"base_jacket_side_back_rt\"),Transform=(Rotation=(X=0,Y=0,Z=0,W=1),Translation=(X=0,Y=-2.9061,Z=2.669),Scale3D=(X=1,Y=1,Z=1))),(Item=(Type=Bone,Name=\"base_jacket_back_cr\"),Transform=(Rotation=(X=0,Y=0,Z=0.0111,W=0.9999),Translation=(X ...(truncated)"
    WeightMinimum=0
    Weight=@n7.OutputParam
   )
   > n8 = &RigUnit_SphericalPoseReader(
    ActiveRegionScaleFactors=(PositiveWidth=1,NegativeWidth=1,PositiveHeight=1,NegativeHeight=1)
    ActiveRegionSize=0.35
    Debug=(bDrawDebug=false,bDraw2D=False,bDrawLocalAxes=False,DebugScale=25,DebugSegments=20,DebugThickness=0.25)
    DriverAxis=(X=-1,Y=0,Z=0)
    DriverItem=(Type=Bone,Name="thigh_l")
    FalloffRegionScaleFactors=(PositiveWidth=1,NegativeWidth=1,PositiveHeight=1,NegativeHeight=1)
    FalloffSize=0.55
    FlipHeightScaling=False
    FlipWidthScaling=False
    OptionalParentItem=(Type=Bone,Name="None")
    RotationOffset=(X=90,Y=0,Z=-20)
   )
   n9 = &RigVMFunction_MathVectorDistance(
    B=&RigUnit_GetTransform(Item=(Type=Bone,Name="calf_l"), bInitial=False).Transform.Translation
    A=&RigUnit_GetTransform(Item=(Type=Bone,Name="spine_03"), bInitial=False).Transform.Translation
   )
   > &RigUnit_ModifyTransforms(
    ItemToModify="((Item=(Type=Bone,Name=\"base_jacket_front_lf\"),Transform=(Rotation=(X=0.2469,Y=-0.2429,Z=-0.4093,W=0.8441),Translation=(X=0,Y=0,Z=0),Scale3D=(X=1,Y=1,Z=1))),(Item=(Type=Bone,Name=\"base_jacket_side_front_lf\"),Transform=(Rotation=(X=0,Y=0,Z=0.1116,W=0.9938), ...(truncated)"
    WeightMinimum=0
    Weight=&RigVMFunction_MathDoubleMul(
     B=&RigVMFunction_MathDoubleRemap(SourceMaximum=39.2818, SourceMinimum=63.8854, TargetMaximum=1, TargetMinimum=0, Value=@n9.Result).Result
     A=@n8.OutputParam
    ).Result
   )
   > n10 = &RigUnit_SphericalPoseReader(
    ActiveRegionScaleFactors=(PositiveWidth=1,NegativeWidth=1,PositiveHeight=1,NegativeHeight=1)
    ActiveRegionSize=0.05
    Debug=(bDrawDebug=false,bDraw2D=False,bDrawLocalAxes=False,DebugScale=25,DebugSegments=20,DebugThickness=0.25)
    DriverAxis=(X=-1,Y=0,Z=0)
    DriverItem=(Type=Bone,Name="thigh_l")
    FalloffRegionScaleFactors=(PositiveWidth=1,NegativeWidth=1,PositiveHeight=1,NegativeHeight=1)
    FalloffSize=0.4
    FlipHeightScaling=False
    FlipWidthScaling=False
    OptionalParentItem=(Type=Bone,Name="None")
    RotationOffset=(X=180,Y=20,Z=0)
   )
   > &RigUnit_ModifyTransforms(
    ItemToModify="((Item=(Type=Bone,Name=\"base_jacket_side_front_lf\"),Transform=(Rotation=(X=0,Y=0,Z=-0.7071,W=0.7071),Translation=(X=0,Y=0,Z=0),Scale3D=(X=1,Y=1,Z=1))),(Item=(Type=Bone,Name=\"base_jacket_side_back_lf\"),Transform=(Rotation=(X=0,Y=0,Z=0,W=1),Translation=(X=0, ...(truncated)"
    WeightMinimum=0
    Weight=@n10.OutputParam
   )
   > n11 = &RigUnit_SphericalPoseReader(
    ActiveRegionScaleFactors=(PositiveWidth=1,NegativeWidth=1,PositiveHeight=1,NegativeHeight=1)
    ActiveRegionSize=0.12
    Debug=(bDrawDebug=false,bDraw2D=False,bDrawLocalAxes=False,DebugScale=25,DebugSegments=20,DebugThickness=0.25)
    DriverAxis=(X=-1,Y=0,Z=0)
    DriverItem=(Type=Bone,Name="thigh_l")
    FalloffRegionScaleFactors=(PositiveWidth=1,NegativeWidth=1,PositiveHeight=1,NegativeHeight=1)
    FalloffSize=0.35
    FlipHeightScaling=False
    FlipWidthScaling=False
    OptionalParentItem=(Type=Bone,Name="None")
    RotationOffset=(X=160,Y=30,Z=0)
   )
   > &RigUnit_ModifyTransforms(
    ItemToModify="((Item=(Type=Bone,Name=\"base_jacket_front_lf\"),Transform=(Rotation=(X=-0.0691,Y=-0.0531,Z=-0.6064,W=0.7903),Translation=(X=0,Y=0,Z=0),Scale3D=(X=1,Y=1,Z=1))),(Item=(Type=Bone,Name=\"base_jacket_side_front_lf\"),Transform=(Rotation=(X=0,Y=-0.2588,Z=0,W=0.9659 ...(truncated)"
    WeightMinimum=0
    Weight=@n11.OutputParam
   )
   > n12 = &RigUnit_SphericalPoseReader(
    ActiveRegionScaleFactors=(PositiveWidth=1,NegativeWidth=1,PositiveHeight=1,NegativeHeight=1)
    ActiveRegionSize=0.45
    Debug=(bDrawDebug=false,bDraw2D=False,bDrawLocalAxes=False,DebugScale=25,DebugSegments=20,DebugThickness=0.25)
    DriverAxis=(X=1,Y=0,Z=0)
    DriverItem=(Type=Bone,Name="thigh_r")
    FalloffRegionScaleFactors=(PositiveWidth=1,NegativeWidth=1,PositiveHeight=1,NegativeHeight=1)
    FalloffSize=0.55
    FlipHeightScaling=true
    FlipWidthScaling=true
    OptionalParentItem=(Type=Bone,Name="None")
    RotationOffset=(X=-90,Y=0,Z=-20)
   )
   RigUnit_GetTransform = &RigUnit_GetTransform(Item=(Type=Bone,Name="calf_r"), bInitial=False)
   RigVMFunction_MathVectorDistance = &RigVMFunction_MathVectorDistance(
    B=@RigUnit_GetTransform.Transform.Translation
    A=&RigUnit_GetTransform(Item=(Type=Bone,Name="spine_03"), bInitial=False).Transform.Translation
   )
   RigVMFunction_AnimEvalRichCurve = &RigVMFunction_AnimEvalRichCurve(
    Curve=(EditorCurveData=(Keys=((InterpMode=RCIM_Cubic,TangentMode=RCTM_User,ArriveTangent=2.9286,LeaveTangent=2.9286),(Time=1,Value=1))))
    SourceMinimum=0
    TargetMinimum=0
    Value=&RigVMFunction_MathDoubleRemap(SourceMaximum=39.2818, SourceMinimum=63.8854, TargetMaximum=1, TargetMinimum=0, Value=@RigVMFunction_MathVectorDistance.Result).Result
   )
   Multiply = &RigVMFunction_MathDoubleMul(B=@RigVMFunction_AnimEvalRichCurve.Result, A=@n12.OutputParam)
   > &RigUnit_ModifyTransforms(
    ItemToModify="((Item=(Type=Bone,Name=\"base_jacket_front_rt\"),Transform=(Rotation=(X=-0.0057,Y=0.2241,Z=-0.0074,W=-0.9745),Translation=(X=0,Y=0,Z=0),Scale3D=(X=1,Y=1,Z=1))),(Item=(Type=Bone,Name=\"base_jacket_side_front_rt\"),Transform=(Rotation=(X=0.0086,Y=-0.0222,Z=0.016 ...(truncated)"
    WeightMinimum=0
    Weight=@Multiply.Result
   )
   > n13 = &RigUnit_SphericalPoseReader(
    ActiveRegionScaleFactors=(PositiveWidth=1,NegativeWidth=1,PositiveHeight=1,NegativeHeight=1)
    ActiveRegionSize=0.05
    Debug=(bDrawDebug=false,bDraw2D=False,bDrawLocalAxes=False,DebugScale=25,DebugSegments=20,DebugThickness=0.25)
    DriverAxis=(X=1,Y=0,Z=0)
    DriverItem=(Type=Bone,Name="thigh_r")
    FalloffRegionScaleFactors=(PositiveWidth=1,NegativeWidth=1,PositiveHeight=1,NegativeHeight=1)
    FalloffSize=0.35
    FlipHeightScaling=true
    FlipWidthScaling=true
    OptionalParentItem=(Type=Bone,Name="None")
    RotationOffset=(X=0,Y=20,Z=0)
   )
   > &RigUnit_ModifyTransforms(
    ItemToModify="((Item=(Type=Bone,Name=\"base_jacket_side_front_rt\"),Transform=(Rotation=(X=0,Y=0,Z=-0.866,W=0.5),Translation=(X=0,Y=0,Z=0),Scale3D=(X=1,Y=1,Z=1))),(Item=(Type=Bone,Name=\"base_jacket_front_rt\"),Transform=(Rotation=(X=0,Y=0,Z=-0.2588,W=0.9659),Translation=(X ...(truncated)"
    WeightMinimum=0
    Weight=&RigVMFunction_MathDoubleMul(B=@RigVMFunction_AnimEvalRichCurve.Result, A=@n13.OutputParam).Result
   )
   > n14 = &RigUnit_SphericalPoseReader(
    ActiveRegionScaleFactors=(PositiveWidth=1,NegativeWidth=1,PositiveHeight=1,NegativeHeight=1)
    ActiveRegionSize=0.05
    Debug=(bDrawDebug=false,bDraw2D=False,bDrawLocalAxes=False,DebugScale=25,DebugSegments=20,DebugThickness=0.25)
    DriverAxis=(X=1,Y=0,Z=0)
    DriverItem=(Type=Bone,Name="thigh_r")
    FalloffRegionScaleFactors=(PositiveWidth=1,NegativeWidth=1,PositiveHeight=0.7,NegativeHeight=0.7)
    FalloffSize=0.35
    FlipHeightScaling=true
    FlipWidthScaling=true
    OptionalParentItem=(Type=Bone,Name="None")
    RotationOffset=(X=-40,Y=20,Z=0)
   )
   > &RigUnit_ModifyTransforms(
    ItemToModify="((Item=(Type=Bone,Name=\"base_jacket_front_rt\"),Transform=(Rotation=(X=0.1113,Y=0.1326,Z=0.6661,W=-0.7255),Translation=(X=0,Y=0,Z=0),Scale3D=(X=1,Y=1,Z=1))),(Item=(Type=Bone,Name=\"base_jacket_side_front_rt\"),Transform=(Rotation=(X=0,Y=0,Z=0,W=1),Translation ...(truncated)"
    WeightMinimum=0
    Weight=&RigVMFunction_MathDoubleMul(B=@RigVMFunction_AnimEvalRichCurve.Result, A=@n14.OutputParam).Result
   )
   &RigUnit_ParentConstraint(
    AdvancedSettings=(RotationOrderForFilter=XZY)
    Child=(Type=Bone,Name="base_necklace_01_cr")
    Parents="((Item=(Type=Bone,Name=\"clavicle_l\"),Weight=0.0051),(Item=(Type=Bone,Name=\"clavicle_r\"),Weight=0.0066),(Item=(Type=Bone,Name=\"head\"),Weight=0.0424),(Item=(Type=Bone,Name=\"neck_01\"),Weight=0.1736),(Item=(Type=Bone,Name=\"neck_02\"),Weight=0.1943),(Item=(Type= ...(truncated)"
    Weight=1
    ExecutePin=@RigVMFunction_Sequence.B
   )
   > &RigUnit_ParentConstraint(
    AdvancedSettings=(RotationOrderForFilter=XZY)
    Child=(Type=Bone,Name="aim_necklace_01_cr")
    Parents=((Item=(Type=Bone,Name="deform_pec_r"),Weight=0.0103),(Item=(Type=Bone,Name="deform_pec_l"),Weight=0.02),(Item=(Type=Bone,Name="spine_05"),Weight=0.1371),(Item=(Type=Bone,Name="spine_04"),Weight=0.8327))
    Weight=1
   )
   > &RigUnit_AimConstraintLocalSpaceOffset(
    Child=(Type=Bone,Name="base_necklace_01_cr")
    Filter=(bX=True,bY=True,bZ=True)
    Parents=((Item=(Type=Bone,Name="aim_necklace_01_cr"),Weight=1))
    Weight=1
    WorldUp=(Target=(X=0,Y=0,Z=1),Kind=Direction,Space=(Type=Bone,Name="aim_necklace_01_cr"))
   )
   > &RigUnit_ParentConstraint(
    AdvancedSettings=(RotationOrderForFilter=XZY)
    Child=(Type=Bone,Name="base_charm_01_rt")
    Parents=((Item=(Type=Bone,Name="thigh_twist_01_r"),Weight=0.1621),(Item=(Type=Bone,Name="thigh_r"),Weight=0.8379))
    Weight=1
   )
   n15 = &RigVMFunction_AnimEvalRichCurve(
    Curve=(EditorCurveData=(Keys=((),(Time=0.5,Value=1),(Time=0.8))))
    SourceMaximum=2
    SourceMinimum=0
    TargetMinimum=0
    Value=&RigVMFunction_MathDoubleRemap(
     SourceMaximum=70
     SourceMinimum=45
     TargetMaximum=1
     TargetMinimum=0
     Value=&RigUnit_GetRelativeTransformForItem(Child=(Type=Bone,Name="pelvis"), Parent=(Type=Bone,Name="root"), bChildInitial=False, bParentInitial=False).RelativeTransform.Translation.Z
    ).Result
   )
   n16 = RigVMVariableNode() {variable=IsCrouching}
   > &RigUnit_ModifyTransforms(
    ItemToModify="((Item=(Type=Bone,Name=\"base_jacket_front_lf\"),Transform=(Rotation=(X=0,Y=0.1736,Z=0,W=0.9848),Translation=(X=0,Y=0,Z=0),Scale3D=(X=1,Y=1,Z=1))),(Item=(Type=Bone,Name=\"base_jacket_side_front_lf\"),Transform=(Rotation=(X=0,Y=0,Z=0.0872,W=0.9962),Translation= ...(truncated)"
    WeightMinimum=0
    Weight=&RigVMFunction_MathDoubleMul(B=@n15.Result, A=&RigVMFunction_MathBoolToFloat(Value=@n16.Value).Result).Result
   )
   > &RigUnit_ModifyTransforms(
    ItemToModify="((Item=(Type=Bone,Name=\"base_jacket_front_lf\"),Transform=(Rotation=(X=-0.0151,Y=0.173,Z=0.0858,W=0.9811),Translation=(X=0,Y=0,Z=0),Scale3D=(X=1,Y=1,Z=1))),(Item=(Type=Bone,Name=\"base_jacket_side_front_lf\"),Transform=(Rotation=(X=0,Y=0,Z=-0.0872,W=0.9962),T ...(truncated)"
    WeightMinimum=0
    Weight=&RigVMFunction_MathDoubleMul(
     B=&RigVMFunction_MathDoubleRemap(SourceMaximum=0, SourceMinimum=1, TargetMaximum=1, TargetMinimum=0, Value=@n15.Result).Result
     A=@n15.Result
    ).Result
   )
   > neck shirt = RigVMFunctionReferenceNode()
   > &RigUnit_ParentConstraint(
    AdvancedSettings=(RotationOrderForFilter=XZY)
    Child=(Type=Bone,Name="base_shirt_corr_cr")
    Parents="((Item=(Type=Bone,Name=\"deform_groin_r\"),Weight=0.0002),(Item=(Type=Bone,Name=\"thigh_l\"),Weight=0.0066),(Item=(Type=Bone,Name=\"thigh_r\"),Weight=0.0116),(Item=(Type=Bone,Name=\"deform_glute_l\"),Weight=0.1112),(Item=(Type=Bone,Name=\"thigh_twist_01_l\"),Weight= ...(truncated)"
    Weight=1
   )
   n17 = RigVMDispatchNode(BoneName=root, Default=0, Name=deformRig_r_thigh_up_alpha) {template="DISPATCH_RigDispatch_GetAnimAttribute(in Name,in BoneName,in Default,out Value,out Found)"}
   n18 = RigVMDispatchNode(BoneName=root, Default=0, Name=deformRig_l_thigh_up_alpha) {template="DISPATCH_RigDispatch_GetAnimAttribute(in Name,in BoneName,in Default,out Value,out Found)"}
   LessEqual = /Script/RigVM.RigVMFunction_MathDoubleLessEqual(B=@n17.Value, A=@n18.Value) {is=struct RigVMUnitNode, template="LessEqual::Execute(in A,in B,out Result)"}
   > &RigUnit_ModifyTransforms(
    ItemToModify=((Item=(Type=Bone,Name="shirt_corr_cr"),Transform=(Rotation=(X=0,Y=0,Z=0,W=-1),Translation=(X=5.0749,Y=-3.7724,Z=0),Scale3D=(X=1,Y=1,Z=1.3305))))
    WeightMinimum=0
    Weight=RigVMDispatchNode(True=@n17.Value, False=@n18.Value, Condition=@LessEqual.Result) {template="DISPATCH_RigVMDispatch_If(in Condition,in True,in False,out Result)"}.Result
   )
   > n19 = &RigUnit_SphericalPoseReader(
    ActiveRegionScaleFactors=(PositiveWidth=1,NegativeWidth=0.5,PositiveHeight=0.75,NegativeHeight=0.75)
    ActiveRegionSize=0.1
    Debug=(bDrawDebug=false,bDraw2D=False,bDrawLocalAxes=False,DebugScale=25,DebugSegments=20,DebugThickness=0.25)
    DriverAxis=(X=0,Y=1,Z=0)
    DriverItem=(Type=Bone,Name="head")
    FalloffRegionScaleFactors=(PositiveWidth=1,NegativeWidth=1,PositiveHeight=1,NegativeHeight=1)
    FalloffSize=0.22
    FlipHeightScaling=False
    FlipWidthScaling=False
    OptionalParentItem=(Type=Bone,Name="None")
    RotationOffset=(X=90,Y=0,Z=-30)
   )
   > n20 = &RigUnit_SphericalPoseReader(
    ActiveRegionScaleFactors=(PositiveWidth=1,NegativeWidth=1,PositiveHeight=1,NegativeHeight=1)
    ActiveRegionSize=0.1
    Debug=(bDrawDebug=false,bDraw2D=False,bDrawLocalAxes=False,DebugScale=25,DebugSegments=20,DebugThickness=0.25)
    DriverAxis=(X=-1,Y=0,Z=0)
    DriverItem=(Type=Bone,Name="upperarm_r")
    FalloffRegionScaleFactors=(PositiveWidth=1,NegativeWidth=1,PositiveHeight=1,NegativeHeight=1)
    FalloffSize=0.2
    FlipHeightScaling=False
    FlipWidthScaling=false
    OptionalParentItem=(Type=Bone,Name="None")
    RotationOffset=(X=-115,Y=0,Z=25)
   )
   > &RigUnit_ModifyTransforms(
    ItemToModify=((Item=(Type=Bone,Name="dyn_neck_shirt_rt"),Transform=(Rotation=(X=0.154,Y=-0.3164,Z=-0.0147,W=0.9359),Translation=(X=0.624,Y=1.189,Z=-0.5501),Scale3D=(X=1,Y=1,Z=1))))
    WeightMinimum=0
    Weight=&RigVMFunction_MathDoubleMul(B=@n19.OutputParam, A=@n20.OutputParam).Result
   )
   &RigUnit_ModifyTransforms(
    ExecutePin=()
    ItemToModify="((Item=(Type=Bone,Name=\"base_shoelace_lower_rt\"),Transform=(Rotation=(X=0.0262,Y=-0.0831,Z=0.2996,W=0.9501),Translation=(X=0,Y=0,Z=0),Scale3D=(X=1,Y=1,Z=1))),(Item=(Type=Bone,Name=\"base_bow_shoelace_lower_rt\"),Transform=(Rotation=(X=0.0346,Y=0.6082,Z=-0.02 ...(truncated)"
    Weight=1
    WeightMinimum=0
   )
  ]}
]
hierarchy [
 root {loc=<0 0 0>, rot=<0 0 0 1>}
 pelvis {parent=root, loc=<0 -0.5301 98.0571>, rot=<0.0127 0.707 -0.0127 -0.707>}
 spine_01 {parent=pelvis, loc=<6.1208 -0.9328 0>, rot=<0 0 0.034 -0.9994>}
 spine_02 {parent=spine_01, loc=<6.5997 -0.2095 0>, rot=<0 0 0.0238 -0.9997>}
 spine_03 {parent=spine_02, loc=<6.4914 -0.1237 0>, rot=<0 0 0.0149 -0.9999>}
 spine_04 {parent=spine_03, loc=<6.9598 0.0794 0>, rot=<0 0 -0.0132 -0.9999>}
 spine_05 {parent=spine_04, loc=<6.9987 0 0>, rot=<0 0 -0.0015 -1>}
 neck_01 {parent=spine_05, loc=<11.3138 0.1529 0>, rot=<0 0 -0.2003 -0.9797>}
 neck_02 {parent=neck_01, loc=<4.5627 0 0>, rot=<0 0 0.0504 -0.9987>}
 head {parent=neck_02, loc=<5.3914 0 0>, rot=<0 0 0.115 -0.9934>}
 base_hair_01_cr {parent=head, loc=<6.3434 9.4089 -0.8199>, rot=<0.2248 0.0139 0.9743 0.0024>}
 dyn_hair_01_cr {parent=base_hair_01_cr, loc=<0 0 0>, rot=<0 0 0 1>}
 base_hair_01_lf {parent=head, loc=<6.6848 3.7173 -6.2133>, rot=<-0.0005 0.4555 0.887 0.0759>}
 dyn_hair_01_lf {parent=base_hair_01_lf, loc=<0 0 0>, rot=<0 0 0 1>}
 base_hair_02_lf {parent=head, loc=<6.3834 0.405 -7.9615>, rot=<-0.0345 0.6398 0.7634 0.0821>}
 dyn_hair_02_lf {parent=base_hair_02_lf, loc=<0 0 0>, rot=<0 0 0 1>}
 base_hair_01_rt {parent=head, loc=<4.6567 7.02 6.1424>, rot=<-0.1077 -0.3213 0.9387 0.063>}
 dyn_hair_01_rt {parent=base_hair_01_rt, loc=<0 0 0>, rot=<0 0 0 1>}
 base_hair_02_rt {parent=head, loc=<4.5491 4.6261 7.0187>, rot=<-0.0383 -0.3441 0.9373 0.0391>}
 dyn_hair_02_rt {parent=base_hair_02_rt, loc=<0 0 0>, rot=<0 0 0 1>}
 base_hair_03_rt {parent=head, loc=<4.0322 0.3647 8.2798>, rot=<-0.0086 -0.6539 0.7564 -0.012>}
 dyn_hair_03_rt {parent=base_hair_03_rt, loc=<0 0 0>, rot=<0 0 0 1>}
 base_hair_back_01_cr {parent=head, loc=<4.0909 -4.1075 0>, rot=<0.2808 -0.9598 0 0>}
 dyn_hair_back_01_cr {parent=base_hair_back_01_cr, loc=<0 0 0>, rot=<0 0 0 1>}
 dyn_hair_back_02_cr {parent=dyn_hair_back_01_cr, loc=<6.253 0 0>, rot=<0 0 0 1>}
 base_tail_hair_01_lf {parent=dyn_hair_back_02_cr, loc=<4.9352 1.3525 1.128>, rot=<0.4431 0.4097 0.1646 -0.7802>}
 dyn_tail_hair_01_lf {parent=base_tail_hair_01_lf, loc=<0 0 0>, rot=<0 0 0 1>}
 dyn_tail_hair_02_lf {parent=dyn_tail_hair_01_lf, loc=<1.2094 0 0>, rot=<-0.2139 0.2281 0.1368 -0.9399>}
 base_tail_hair_03_lf {parent=dyn_tail_hair_02_lf, loc=<3.4846 0 0>, rot=<0.2309 0.5427 0.0395 -0.8066>}
 dyn_tail_hair_03_lf {parent=base_tail_hair_03_lf, loc=<0 0 0>, rot=<0 0 0 1>}
 dyn_tail_hair_04_lf {parent=dyn_tail_hair_03_lf, loc=<4.5 0 0>, rot=<0.0006 0.1668 -0.2184 -0.9615>}
 dyn_tail_hair_05_lf {parent=dyn_tail_hair_04_lf, loc=<4.5 0 0>, rot=<0.0067 0.1876 -0.3162 -0.9299>}
 dyn_tail_hair_06_lf {parent=dyn_tail_hair_05_lf, loc=<4.5 0 0>, rot=<-0.0116 -0.0453 -0.0994 -0.9939>}
 dyn_tail_hair_07_lf {parent=dyn_tail_hair_06_lf, loc=<4.5 0 0>, rot=<-0.0001 -0.0211 -0.0133 -0.9997>}
 dyn_tail_hair_ik_07_lf {parent=dyn_tail_hair_07_lf, loc=<0 0 0>, rot=<0 0 0 1>}
 dyn_tail_hair_ik_06_lf {parent=dyn_tail_hair_06_lf, loc=<0 0 0>, rot=<0 0 0 1>}
 dyn_tail_hair_ik_05_lf {parent=dyn_tail_hair_05_lf, loc=<0 0 0>, rot=<0 0 0 1>}
 dyn_tail_hair_ik_04_lf {parent=dyn_tail_hair_04_lf, loc=<0 0 0>, rot=<0 0 0 1>}
 dyn_tail_hair_ik_03_lf {parent=dyn_tail_hair_03_lf, loc=<0 0 0>, rot=<0 0 0 1>}
 base_necklace_01_cr {parent=neck_01, loc=<0 0 0>, rot=<0.3884 0.5909 0.5909 0.3884>}
 dyn_necklace_01_cr {parent=base_necklace_01_cr, loc=<0 0 0>, rot=<0 0 0 1>}
 dyn_necklace_02_cr {parent=dyn_necklace_01_cr, loc=<4.2279 0 0>, rot=<0 0 0 1>}
 dyn_necklace_03_cr {parent=dyn_necklace_02_cr, loc=<4.2279 0 0>, rot=<0 0 0 1>}
 dyn_necklace_04_cr {parent=dyn_necklace_03_cr, loc=<4.2279 0 0>, rot=<0 0 0 1>}
 clavicle_l {parent=spine_05, loc=<5.4831 -0.1053 -3.2946>, rot=<0.0465 -0.7289 -0.009 -0.6829>}
 upperarm_l {parent=clavicle_l, loc=<9.8658 -1.2136 -0.4105>, rot=<0.0164 -0.3245 -0.0269 -0.9454>}
 lowerarm_l {parent=upperarm_l, loc=<26.465 -0.8345 0.259>, rot=<0.0036 0.0161 -0.2298 -0.9731>}
 hand_l {parent=lowerarm_l, loc=<22.7355 -0.3203 -0.6105>, rot=<-0.6784 -0.0223 0.0735 -0.7307>}
 thumb_01_l {parent=hand_l, loc=<1.3118 -0.3191 -2.459>, rot=<0.6546 -0.4042 0.0664 -0.6354>}
 thumb_02_l {parent=thumb_01_l, loc=<5.9336 -0.0448 0.0401>, rot=<-0.0037 0.0213 0.298 -0.9543>}
 thumb_03_l {parent=thumb_02_l, loc=<2.2223 0.3682 -0.0907>, rot=<0.0118 -0.013 0.0066 -0.9998>}
 index_metacarpal_l {parent=hand_l, loc=<1.6535 0.0183 -2.96>, rot=<0.0158 -0.0124 -0.0173 -0.9996>}
 index_01_l {parent=index_metacarpal_l, loc=<8.3008 0.0357 -0.0524>, rot=<-0.0044 0.0485 0.0924 -0.9945>}
 index_02_l {parent=index_01_l, loc=<3.3587 -0.3004 0.0452>, rot=<-0.002 0.0027 0.3225 -0.9466>}
 index_03_l {parent=index_02_l, loc=<2.5578 -0.123 -0.0252>, rot=<-0.002 0.0067 0.2558 -0.9667>}
 middle_metacarpal_l {parent=hand_l, loc=<2.1856 -0.1725 -0.5984>, rot=<0.006 0.0168 -0.0329 -0.9993>}
 middle_01_l {parent=middle_metacarpal_l, loc=<7.3275 -0.2447 0.0129>, rot=<-0.0208 0.0042 0.0858 -0.9961>}
 middle_02_l {parent=middle_01_l, loc=<4.0123 0.1401 0.1175>, rot=<-0.0218 0.0258 0.3512 -0.9357>}
 middle_03_l {parent=middle_02_l, loc=<2.6077 0.0534 0.1373>, rot=<0.0082 0.0188 0.1708 -0.9851>}
 ring_metacarpal_l {parent=hand_l, loc=<2.2074 -0.3625 1.3638>, rot=<-0.0284 0.0722 -0.0458 -0.9959>}
 ring_01_l {parent=ring_metacarpal_l, loc=<7.0295 -0.6197 -0.1446>, rot=<-0.0489 -0.0252 0.1892 -0.9804>}
 ring_02_l {parent=ring_01_l, loc=<3.6174 0.4647 -0.1027>, rot=<0 0 0.2806 -0.9598>}
 ring_03_l {parent=ring_02_l, loc=<2.4152 0.114 0.0199>, rot=<0 0 0.1189 -0.9929>}
 pinky_metacarpal_l {parent=hand_l, loc=<1.7616 -0.9753 3.3814>, rot=<-0.0397 0.0814 -0.069 -0.9935>}
 pinky_01_l {parent=pinky_metacarpal_l, loc=<7.1709 -0.4831 -0.1094>, rot=<-0.0654 -0.0035 0.2242 -0.9723>}
 pinky_02_l {parent=pinky_01_l, loc=<2.8274 0.0893 -0.1043>, rot=<0.0037 0.0011 0.2854 -0.9584>}
 pinky_03_l {parent=pinky_02_l, loc=<2.2046 0.0307 -0.0078>, rot=<-0.0002 0.0098 0.1445 -0.9895>}
 deform_wrist_dn_base_l {parent=hand_l, loc=<0 -1.5326 0>, rot=<0 0 0 -1>}
 deform_wrist_dn_l {parent=deform_wrist_dn_base_l, loc=<0 0 0>, rot=<0 0 0 1>}
 deform_wrist_up_base_l {parent=hand_l, loc=<0 2.3009 0>, rot=<0 0 0 -1>}
 deform_wrist_up_l {parent=deform_wrist_up_base_l, loc=<0 0 0>, rot=<0 0 0 1>}
 hand_attach_l {parent=hand_l, loc=<8.5 -3 0>, rot=<0.7071 0.7071 0 0>}
 weapon_l {parent=hand_l, loc=<-0.1974 -3.4111 0.3811>, rot=<0 0 0 1>}
 lowerarm_twist_02_l {parent=lowerarm_l, loc=<10.6707 0 0>, rot=<0 0 0 1>}
 lowerarm_twist_01_l {parent=lowerarm_l, loc=<18.316 0 0>, rot=<0 0 0 1>}
 deform_elbow_in_base_l {parent=lowerarm_l, loc=<0.1834 3.6062 -0.2572>, rot=<0 0 0 -1>}
 deform_elbow_in_l {parent=deform_elbow_in_base_l, loc=<0 0 0>, rot=<0 0 0 1>}
 deform_elbow_out_base_l {parent=lowerarm_l, loc=<0 -4 0>, rot=<0 0 0 1>}
 deform_elbow_out_l {parent=deform_elbow_out_base_l, loc=<0 0 0>, rot=<0 0 0 1>}
 upperarm_twist_01_l {parent=upperarm_l, loc=<7.2135 -0.1134 0>, rot=<0 0 0 1>}
 deform_upperarm_bk_base_l {parent=upperarm_twist_01_l, loc=<-6.4053 -4.3746 0.314>, rot=<0 0 0 -1>}
 deform_upperarm_bk_l {parent=deform_upperarm_bk_base_l, loc=<0 0 0>, rot=<0 0 0 1>}
 deform_upperarm_fr_base_l {parent=upperarm_twist_01_l, loc=<-5.6312 5.092 0.3354>, rot=<0 0 0 -1>}
 deform_upperarm_fr_l {parent=deform_upperarm_fr_base_l, loc=<0 0 0>, rot=<0 0 0 1>}
 deform_upperarm_up_base_l {parent=upperarm_twist_01_l, loc=<-8.891 0.4825 3.9517>, rot=<0 0 0 -1>}
 deform_upperarm_up_l {parent=deform_upperarm_up_base_l, loc=<0 0 0>, rot=<0 0 0 1>}
 null1 {parent=upperarm_twist_01_l, loc=<0 0 0>, rot=<0 0 0 1>}
 upperarm_twist_02_l {parent=upperarm_l, loc=<19.7998 -0.1134 0>, rot=<0 0 0 1>}
 base_upperarm_sleeve_lf {parent=upperarm_l, loc=<18.4655 0 0>, rot=<0.0037 -1 0.0044 0.0006>}
 dyn_upperarm_sleeve_lf {parent=base_upperarm_sleeve_lf, loc=<0 0 0>, rot=<0 0 0 1>}
 dyn_tail_hair_for_aim {parent=clavicle_l, loc=<5.3056 3.4484 10.8631>, rot=<0.3511 0.3921 -0.4359 -0.73>}
 clavicle_r {parent=spine_05, loc=<5.4831 -0.1053 3.2946>, rot=<0.7289 0.0465 -0.6829 0.009>}
 upperarm_r {parent=clavicle_r, loc=<-9.8658 1.2136 0.4105>, rot=<0.0164 -0.3245 -0.0269 -0.9454>}
 lowerarm_r {parent=upperarm_r, loc=<-26.465 0.8345 -0.259>, rot=<0.0036 0.0161 -0.2298 -0.9731>}
 hand_r {parent=lowerarm_r, loc=<-22.7355 0.3203 0.6106>, rot=<-0.6784 -0.0223 0.0735 -0.7307>}
 thumb_01_r {parent=hand_r, loc=<-1.3118 0.3191 2.459>, rot=<0.6546 -0.4042 0.0664 -0.6354>}
 thumb_02_r {parent=thumb_01_r, loc=<-5.9336 0.0448 -0.0401>, rot=<-0.0037 0.0213 0.298 -0.9543>}
 thumb_03_r {parent=thumb_02_r, loc=<-2.2223 -0.3682 0.0907>, rot=<0.0118 -0.013 0.0066 -0.9998>}
 index_metacarpal_r {parent=hand_r, loc=<-1.6535 -0.0183 2.96>, rot=<0.0158 -0.0124 -0.0173 -0.9996>}
 index_01_r {parent=index_metacarpal_r, loc=<-8.3008 -0.0357 0.0524>, rot=<-0.0044 0.0485 0.0924 -0.9945>}
 index_02_r {parent=index_01_r, loc=<-3.3587 0.3004 -0.0452>, rot=<-0.002 0.0027 0.3225 -0.9466>}
 index_03_r {parent=index_02_r, loc=<-2.5578 0.123 0.0252>, rot=<-0.002 0.0067 0.2558 -0.9667>}
 middle_metacarpal_r {parent=hand_r, loc=<-2.1856 0.1725 0.5984>, rot=<0.006 0.0168 -0.0329 -0.9993>}
 middle_01_r {parent=middle_metacarpal_r, loc=<-7.3275 0.2447 -0.0129>, rot=<-0.0208 0.0042 0.0858 -0.9961>}
 middle_02_r {parent=middle_01_r, loc=<-4.0123 -0.1401 -0.1175>, rot=<-0.0218 0.0258 0.3512 -0.9357>}
 middle_03_r {parent=middle_02_r, loc=<-2.6077 -0.0534 -0.1373>, rot=<0.0082 0.0188 0.1708 -0.9851>}
 ring_metacarpal_r {parent=hand_r, loc=<-2.2074 0.3625 -1.3638>, rot=<-0.0284 0.0722 -0.0458 -0.9959>}
 ring_01_r {parent=ring_metacarpal_r, loc=<-7.0295 0.6197 0.1446>, rot=<-0.0489 -0.0252 0.1892 -0.9804>}
 ring_02_r {parent=ring_01_r, loc=<-3.6174 -0.4647 0.1027>, rot=<0 0 0.2806 -0.9598>}
 ring_03_r {parent=ring_02_r, loc=<-2.4152 -0.114 -0.0199>, rot=<0 0 0.1189 -0.9929>}
 pinky_metacarpal_r {parent=hand_r, loc=<-1.7616 0.9753 -3.3814>, rot=<-0.0397 0.0814 -0.069 -0.9935>}
 pinky_01_r {parent=pinky_metacarpal_r, loc=<-7.1709 0.4831 0.1094>, rot=<-0.0654 -0.0035 0.2242 -0.9723>}
 pinky_02_r {parent=pinky_01_r, loc=<-2.8274 -0.0893 0.1043>, rot=<0.0037 0.0011 0.2854 -0.9584>}
 pinky_03_r {parent=pinky_02_r, loc=<-2.2046 -0.0307 0.0078>, rot=<-0.0002 0.0098 0.1445 -0.9895>}
 deform_wrist_dn_base_r {parent=hand_r, loc=<-0.0001 1.5328 0>, rot=<0 0 0 -1>}
 deform_wrist_dn_r {parent=deform_wrist_dn_base_r, loc=<0 0 0>, rot=<0 0 0 1>}
 deform_wrist_up_base_r {parent=hand_r, loc=<-0.0002 -2.3007 -0.0001>, rot=<0 0 0 -1>}
 deform_wrist_up_r {parent=deform_wrist_up_base_r, loc=<0 0 0>, rot=<0 0 0 1>}
 weapon_r {parent=hand_r, loc=<0.1974 3.4111 -0.3811>, rot=<0 0 0 1>}
 hand_attach_r {parent=hand_r, loc=<-8.5001 2.9999 0>, rot=<0 0 -0.7071 -0.7071>}
 lowerarm_twist_02_r {parent=lowerarm_r, loc=<-10.6707 0 0>, rot=<0 0 0 -1>}
 lowerarm_twist_01_r {parent=lowerarm_r, loc=<-18.316 0 0>, rot=<0 0 0 -1>}
 base_bow_hand_rt_lower_01_cr {parent=lowerarm_twist_01_r, loc=<-2.5312 0.3644 0.1031>, rot=<-0.4077 -0.5112 0.5661 -0.502>}
 dyn_bow_hand_rt_lower_01_cr {parent=base_bow_hand_rt_lower_01_cr, loc=<0 0 0>, rot=<0 0 0 1>}
 dyn_bow_hand_rt_lower_rt {parent=dyn_bow_hand_rt_lower_01_cr, loc=<3.7188 -0.3229 -0.5104>, rot=<0 -0.5367 0 -0.8438>}
 dyn_bow_hand_rt_lower_lf {parent=dyn_bow_hand_rt_lower_01_cr, loc=<3.7188 -0.3229 -0.5104>, rot=<0 -0.5367 0 -0.8438>}
 deform_elbow_in_base_r {parent=lowerarm_r, loc=<-0.1834 -3.6062 0.2572>, rot=<0 0 0 -1>}
 deform_elbow_in_r {parent=deform_elbow_in_base_r, loc=<0 0 0>, rot=<0 0 0 1>}
 deform_elbow_out_base_r {parent=lowerarm_r, loc=<0 4 -0.0001>, rot=<0 0 0 -1>}
 deform_elbow_out_r {parent=deform_elbow_out_base_r, loc=<0 0 0>, rot=<0 0 0 1>}
 upperarm_twist_01_r {parent=upperarm_r, loc=<-7.2135 0.1134 0>, rot=<0 0 0 -1>}
 deform_upperarm_bk_base_r {parent=upperarm_twist_01_r, loc=<6.4053 4.3746 -0.314>, rot=<0 0 0 -1>}
 deform_upperarm_bk_r {parent=deform_upperarm_bk_base_r, loc=<0 0 0>, rot=<0 0 0 1>}
 deform_upperarm_fr_base_r {parent=upperarm_twist_01_r, loc=<5.6312 -5.092 -0.3354>, rot=<0 0 0 -1>}
 deform_upperarm_fr_r {parent=deform_upperarm_fr_base_r, loc=<0 0 0>, rot=<0 0 0 1>}
 deform_upperarm_up_base_r {parent=upperarm_twist_01_r, loc=<8.891 -0.4825 -3.9517>, rot=<0 0 0 -1>}
 deform_upperarm_up_r {parent=deform_upperarm_up_base_r, loc=<0 0 0>, rot=<0 0 0 1>}
 upperarm_twist_02_r {parent=upperarm_r, loc=<-19.7998 0.1134 0>, rot=<0 0 0 -1>}
 base_upperarm_sleeve_rt {parent=upperarm_r, loc=<-18.4655 0 0.0001>, rot=<0.0044 -0.0006 -0.0037 -1>}
 dyn_upperarm_sleeve_rt {parent=base_upperarm_sleeve_rt, loc=<0 0 0>, rot=<0 0 0 1>}
 attach_backpack {parent=spine_05, loc=<-0.123 -13.4729 0>, rot=<-0.0093 -0.707 0.0093 -0.707>}
 Backpack_BR {parent=attach_backpack, type=Null, loc=<0 3.5868 1.9148>, rot=<-0.0441 0 0 0.999>}
 attach_cape {parent=spine_05, loc=<9.0171 -11.714 0>, rot=<0.0616 -0.7044 -0.0616 -0.7044>}
 cape {parent=attach_cape, type=Null, loc=<0 0 0>, rot=<0 0 0 1>}
 base_neck_shirt_lf {parent=spine_05, loc=<9.5111 4.5144 -7.7962>, rot=<0.1717 -0.6859 0.1162 -0.6975>}
 dyn_neck_shirt_lf {parent=base_neck_shirt_lf, loc=<0 0 0>, rot=<0 0 0 1>}
 base_neck_shirt_rt {parent=spine_05, loc=<9.5108 4.5143 7.7962>, rot=<-0.6859 -0.1717 0.6975 0.1162>}
 dyn_neck_shirt_rt {parent=base_neck_shirt_rt, loc=<0 0 0>, rot=<0 0 0 1>}
 dyn_hair_back_aim {parent=spine_05, loc=<6.9978 -13.7159 0>, rot=<0.3151 -0.949 0 0>}
 aim_necklace_01_cr {parent=spine_05, loc=<-5.0127 15.664 0>, rot=<0.2622 0.6567 0.6567 0.2622>}
 deform_pec_aim_base_l {parent=spine_04, loc=<3.2456 9.3309 -1.2237>, rot=<0.0603 -0.4033 0.3318 -0.8507>}
 deform_pec_aim_l {parent=deform_pec_aim_base_l, loc=<0 0 0>, rot=<0 0 0 1>}
 deform_pec_base_l {parent=deform_pec_aim_l, loc=<9.3665 0 0>, rot=<0 0 0 1>}
 deform_pec_l {parent=deform_pec_base_l, loc=<0 0 0>, rot=<0 0 0 1>}
 deform_pec_aim_base_r {parent=spine_04, loc=<3.2452 9.3309 1.2237>, rot=<-0.4033 -0.0603 0.8507 0.3318>}
 deform_pec_aim_r {parent=deform_pec_aim_base_r, loc=<0 0 0>, rot=<0 0 0 1>}
 deform_pec_base_r {parent=deform_pec_aim_r, loc=<-9.3667 -0.0003 -0.0004>, rot=<0 0 0 1>}
 deform_pec_r {parent=deform_pec_base_r, loc=<0 0 0>, rot=<0 0 0 1>}
 base_jacket_spine_cr {parent=spine_03, loc=<0 0 0>, rot=<0 0 0 1>}
 jacket_spine_cr {parent=base_jacket_spine_cr, loc=<0 0 0>, rot=<0 0 0 1>}
 base_jacket_front_lf {parent=jacket_spine_cr, loc=<-2.8267 10.7462 -12.0868>, rot=<-0.0789 0.5855 0.8068 -0.0047>}
 dyn_jacket_front_01_lf {parent=base_jacket_front_lf, loc=<0 0 0>, rot=<0 0 0 1>}
 dyn_jacket_front_02_lf {parent=dyn_jacket_front_01_lf, loc=<8.2111 0 0>, rot=<0 0 0 1>}
 dyn_jacket_front_03_lf {parent=dyn_jacket_front_02_lf, loc=<8.2111 0 0>, rot=<0 0 0.0387 -0.9992>}
 base_jacket_side_front_lf {parent=jacket_spine_cr, loc=<-1.2413 -1.6355 -15.2308>, rot=<0.0755 -0.8043 -0.5894 0.0094>}
 dyn_jacket_side_front_01_lf {parent=base_jacket_side_front_lf, loc=<0 0 0>, rot=<0 0 0 1>}
 dyn_jacket_side_front_02_lf {parent=dyn_jacket_side_front_01_lf, loc=<8.2111 0 0>, rot=<0 0 0.0981 -0.9952>}
 dyn_jacket_side_front_03_lf {parent=dyn_jacket_side_front_02_lf, loc=<8.2111 0 0>, rot=<0 0 0 1>}
 base_jacket_side_back_lf {parent=jacket_spine_cr, loc=<-0.4575 -9.786 -9.1805>, rot=<0.0408 -0.9812 -0.1876 -0.0179>}
 dyn_jacket_side_back_01_lf {parent=base_jacket_side_back_lf, loc=<0 0 0>, rot=<0 0 0 1>}
 dyn_jacket_side_back_02_lf {parent=dyn_jacket_side_back_01_lf, loc=<8.2111 0 0>, rot=<0 0 0.0808 -0.9967>}
 dyn_jacket_side_back_03_lf {parent=dyn_jacket_side_back_02_lf, loc=<8.2111 0 0>, rot=<0 0 0 1>}
 base_jacket_back_cr {parent=jacket_spine_cr, loc=<-0.4544 -10.4925 0>, rot=<0.0587 -0.9983 0 0>}
 dyn_jacket_back_01_cr {parent=base_jacket_back_cr, loc=<0 0 0>, rot=<0 0 0 1>}
 dyn_jacket_back_02_cr {parent=dyn_jacket_back_01_cr, loc=<8.2111 0 0>, rot=<0 0 0.0941 -0.9956>}
 dyn_jacket_back_03_cr {parent=dyn_jacket_back_02_cr, loc=<8.2111 0 0>, rot=<0 0 0 1>}
 base_jacket_side_back_rt {parent=jacket_spine_cr, loc=<-0.4578 -9.7861 9.1805>, rot=<0.0408 -0.9812 0.1876 0.0179>}
 dyn_jacket_side_back_01_rt {parent=base_jacket_side_back_rt, loc=<0 0 0>, rot=<0 0 0 -1>}
 dyn_jacket_side_back_02_rt {parent=dyn_jacket_side_back_01_rt, loc=<8.2108 0 0>, rot=<0 0 0.0808 -0.9967>}
 dyn_jacket_side_back_03_rt {parent=dyn_jacket_side_back_02_rt, loc=<8.2111 0 0>, rot=<0 0 0 1>}
 base_jacket_side_front_rt {parent=jacket_spine_cr, loc=<-1.2414 -1.6355 15.2308>, rot=<0.0755 -0.8043 0.5894 -0.0094>}
 dyn_jacket_side_front_01_rt {parent=base_jacket_side_front_rt, loc=<0 0 0>, rot=<0 0 0 1>}
 dyn_jacket_side_front_02_rt {parent=dyn_jacket_side_front_01_rt, loc=<8.2107 -0.0001 0>, rot=<0 0 0.0715 -0.9974>}
 dyn_jacket_side_front_03_rt {parent=dyn_jacket_side_front_02_rt, loc=<8.2113 0 0>, rot=<0 0 0 1>}
 base_jacket_front_rt {parent=jacket_spine_cr, loc=<-2.8752 10.2846 11.0614>, rot=<0.0181 -0.4443 0.895 -0.0366>}
 dyn_jacket_front_01_rt {parent=base_jacket_front_rt, loc=<0 0 0>, rot=<0 0 0 1>}
 dyn_jacket_front_02_rt {parent=dyn_jacket_front_01_rt, loc=<8.2115 0.0001 0>, rot=<0.0017 0.0039 0.0187 -0.9998>}
 dyn_jacket_front_03_rt {parent=dyn_jacket_front_02_rt, loc=<8.2111 -0.0001 0>, rot=<0 0 0.0387 -0.9992>}
 spine_03_for_cons_offset {parent=spine_03, loc=<0 0 0>, rot=<0 0 0 1>}
 spine_03_for_cons {parent=spine_03_for_cons_offset, loc=<0 0 0>, rot=<0 0 0 1>}
 spine_02_for_cons_offset {parent=spine_02, loc=<0 0 0>, rot=<0 0 0 1>}
 spine_02_for_cons {parent=spine_02_for_cons_offset, loc=<0 0 0>, rot=<0 0 0.0149 -0.9999>}
 base_shirt_spine_cr {parent=spine_01, loc=<0 0 0>, rot=<0 0 0 1>}
 base_shirt_back_01_lf {parent=base_shirt_spine_cr, loc=<-0.4112 -10.4776 -7.3526>, rot=<0.0856 -0.9963 0 0>}
 dyn_shirt_back_01_lf {parent=base_shirt_back_01_lf, loc=<0 0 0>, rot=<0 0 0 1>}
 dyn_shirt_back_02_lf {parent=dyn_shirt_back_01_lf, loc=<5.487 0 0>, rot=<0 0 0 1>}
 dyn_shirt_back_03_lf {parent=dyn_shirt_back_02_lf, loc=<5.4871 0 0>, rot=<0 0 0 1>}
 base_shirt_back_01_rt {parent=base_shirt_spine_cr, loc=<-0.4112 -10.4776 7.353>, rot=<0.0856 -0.9963 0 0>}
 dyn_shirt_back_01_rt {parent=base_shirt_back_01_rt, loc=<0 0 0>, rot=<0 0 0 1>}
 dyn_shirt_back_02_rt {parent=dyn_shirt_back_01_rt, loc=<5.487 0 0>, rot=<0 0 0 1>}
 dyn_shirt_back_03_rt {parent=dyn_shirt_back_02_rt, loc=<5.4871 0 0>, rot=<0 0 0 1>}
 base_shirt_front_01_rt {parent=base_shirt_spine_cr, loc=<-0.9982 7.7921 9.5032>, rot=<0 0 0.9999 0.0172>}
 dyn_shirt_front_01_rt {parent=base_shirt_front_01_rt, loc=<0 0 0>, rot=<0 0 0 1>}
 dyn_shirt_front_02_rt {parent=dyn_shirt_front_01_rt, loc=<5.487 0 0>, rot=<0 0 0 1>}
 dyn_shirt_front_03_rt {parent=dyn_shirt_front_02_rt, loc=<5.4871 0 0>, rot=<0 0 0 1>}
 base_shirt_side_01_rt {parent=base_shirt_spine_cr, loc=<-0.7177 -0.6519 14.5281>, rot=<0.052 -0.7052 0.703 0.0763>}
 dyn_shirt_side_01_rt {parent=base_shirt_side_01_rt, loc=<0 0 0>, rot=<0 0 0 1>}
 dyn_shirt_side_02_rt {parent=dyn_shirt_side_01_rt, loc=<5.487 0 0>, rot=<0 0 0 1>}
 dyn_shirt_side_03_rt {parent=dyn_shirt_side_02_rt, loc=<5.487 -0.0001 0>, rot=<0 0 0 1>}
 base_shirt_side_01_lf {parent=base_shirt_spine_cr, loc=<-0.7177 -0.6519 -14.5281>, rot=<0.052 -0.7052 -0.703 -0.0763>}
 dyn_shirt_side_01_lf {parent=base_shirt_side_01_lf, loc=<0 0 0>, rot=<0 0 0 1>}
 dyn_shirt_side_02_lf {parent=dyn_shirt_side_01_lf, loc=<5.487 0 0>, rot=<0 0 0 1>}
 dyn_shirt_side_03_lf {parent=dyn_shirt_side_02_lf, loc=<5.4871 0 0>, rot=<0 0 0 1>}
 base_shirt_front_01_lf {parent=base_shirt_spine_cr, loc=<-0.9979 7.7921 -9.5032>, rot=<0 0 0.9999 0.0172>}
 dyn_shirt_front_01_lf {parent=base_shirt_front_01_lf, loc=<0 0 0>, rot=<0 0 0 1>}
 dyn_shirt_front_02_lf {parent=dyn_shirt_front_01_lf, loc=<5.487 0 0>, rot=<0 0 0 1>}
 dyn_shirt_front_03_lf {parent=dyn_shirt_front_02_lf, loc=<5.4871 0 0>, rot=<0 0 0 1>}
 base_shirt_front_cr {parent=base_shirt_spine_cr, loc=<-1.185 10.4932 -0.011>, rot=<0 0 0.9994 -0.0359>}
 dyn_shirt_front_01_cr {parent=base_shirt_front_cr, loc=<0 0 0>, rot=<0 0 0 1>}
 dyn_shirt_front_02_cr {parent=dyn_shirt_front_01_cr, loc=<5.487 0 0>, rot=<0 0 0 1>}
 dyn_shirt_front_03_cr {parent=dyn_shirt_front_02_cr, loc=<5.4871 0 0>, rot=<0 0 0 1>}
 spine_01_for_cons_offset {parent=spine_01, loc=<0 0 0>, rot=<0 0 0 1>}
 spine_01_for_cons {parent=spine_01_for_cons_offset, loc=<0 0 0>, rot=<0 0 0 1>}
 thigh_l {parent=pelvis, loc=<-1.0855 -0.2109 -8.7053>, rot=<0.0009 0.0355 0.0155 -0.9993>}
 calf_l {parent=thigh_l, loc=<-42.5284 -0.0019 -0.7297>, rot=<0.0006 -0.0161 -0.0391 -0.9991>}
 foot_l {parent=calf_l, loc=<-46.7401 -2.5014 -0.592>, rot=<0.0427 -0.0238 0.0353 -0.9982>}
 ball_l {parent=foot_l, loc=<-3.5658 10.8066 -0.1544>, rot=<0.0001 -0.0001 -0.7063 -0.7079>}
 base_main_shoelaces_lf {parent=foot_l, loc=<6.5697 7.7825 0.0686>, rot=<0.5234 0.4729 0.5208 0.4807>}
 dyn_main_shoelaces_lf {parent=base_main_shoelaces_lf, loc=<0 0 0>, rot=<0 0 0 1>}
 base_bow_shoelace_upper_lf {parent=dyn_main_shoelaces_lf, loc=<0.243 -0.1671 0.2136>, rot=<0.1985 0.1985 0.6787 -0.6787>}
 dyn_bow_shoelace_upper_lf {parent=base_bow_shoelace_upper_lf, loc=<0 0 0>, rot=<0 0 0 1>}
 base_bow_shoelace_lower_lf {parent=dyn_main_shoelaces_lf, loc=<0.243 -0.1671 0.2136>, rot=<-0.1774 -0.1774 0.6845 -0.6845>}
 dyn_bow_shoelace_lower_lf {parent=base_bow_shoelace_lower_lf, loc=<0 0 0>, rot=<0 0 0 1>}
 base_shoelace_lower_lf {parent=dyn_main_shoelaces_lf, loc=<0.1472 -0.7457 -1.1654>, rot=<-0.1298 -0.4203 0.2506 -0.8624>}
 dyn_shoelace_lower_01_lf {parent=base_shoelace_lower_lf, loc=<0 0 0>, rot=<0 0 0 1>}
 dyn_shoelace_lower_02_lf {parent=dyn_shoelace_lower_01_lf, loc=<5.7484 0 0>, rot=<0 0 0 1>}
 base_shoelace_lower_03_lf {parent=dyn_main_shoelaces_lf, loc=<0.176 0.9925 -0.1127>, rot=<0.017 -0.5286 -0.4124 -0.7417>}
 dyn_shoelace_lower_03_lf {parent=base_shoelace_lower_03_lf, loc=<0 0 0>, rot=<0 0 0 1>}
 dyn_shoelace_lower_04_lf {parent=dyn_shoelace_lower_03_lf, loc=<5.7181 0 0>, rot=<0 0 0.1392 -0.9903>}
 calf_twist_02_l {parent=calf_l, loc=<-19.1546 -1.7943 -0.2638>, rot=<0.0237 0.0095 -0.0035 -0.9997>}
 calf_twist_01_l {parent=calf_l, loc=<-35.8483 -1.9197 -0.5794>, rot=<0.0237 0.0095 -0.0035 -0.9997>}
 deform_knee_in_base_l {parent=calf_l, loc=<0.8031 -5.0231 0.2456>, rot=<0 0 0 -1>}
 deform_knee_in_l {parent=deform_knee_in_base_l, loc=<0 0 0>, rot=<0 0 0 1>}
 deform_knee_out_base_l {parent=calf_l, loc=<0 4 0>, rot=<0 0 0 -1>}
 deform_knee_out_l {parent=deform_knee_out_base_l, loc=<0 0 0>, rot=<0 0 0 1>}
 thigh_twist_01_l {parent=thigh_l, loc=<-27.952 0 0>, rot=<0.0013 0 0 -1>}
 deform_groin_base_l {parent=thigh_l, loc=<0.4801 5.6776 -0.2486>, rot=<0 0 0 -1>}
 deform_groin_l {parent=deform_groin_base_l, loc=<0 0 0>, rot=<0 0 0 1>}
 thigh_r {parent=pelvis, loc=<-1.0855 -0.2109 8.7053>, rot=<0.0355 -0.0009 0.9993 0.0155>}
 calf_r {parent=thigh_r, loc=<42.5284 0.0019 0.7297>, rot=<0.0006 -0.0161 -0.0391 -0.9991>}
 foot_r {parent=calf_r, loc=<46.7401 2.5014 0.592>, rot=<0.0427 -0.0238 0.0353 -0.9982>}
 ball_r {parent=foot_r, loc=<3.5658 -10.8066 0.1544>, rot=<0.0001 -0.0001 -0.7063 -0.7079>}
 base_main_shoelaces_rt {parent=foot_r, loc=<-6.6742 -7.7998 -0.1623>, rot=<-0.5208 0.4807 0.5234 -0.4729>}
 dyn_main_shoelaces_rt {parent=base_main_shoelaces_rt, loc=<0 0 0>, rot=<0 0 0 1>}
 base_bow_shoelace_upper_rt {parent=dyn_main_shoelaces_rt, loc=<-0.0992 0.2367 0.1021>, rot=<0.3061 0.6515 -0.4579 -0.5218>}
 dyn_bow_shoelace_upper_rt {parent=base_bow_shoelace_upper_rt, loc=<0 0 0>, rot=<0 0 0 1>}
 base_shoelace_mid_rt {parent=dyn_main_shoelaces_rt, loc=<-0.1526 3.1244 -3.2878>, rot=<-0.5056 0.131 0.8059 0.2788>}
 dyn_shoelace_mid_rt {parent=base_shoelace_mid_rt, loc=<0 0 0>, rot=<0 0 0 1>}
 base_shoelace_lower_rt {parent=dyn_main_shoelaces_rt, loc=<0.1511 -0.1077 -1.2894>, rot=<0.2624 -0.3155 -0.4732 -0.7796>}
 dyn_shoelace_lower_01_rt {parent=base_shoelace_lower_rt, loc=<0 0 0>, rot=<0 0 0 1>}
 dyn_shoelace_lower_02_rt {parent=dyn_shoelace_lower_01_rt, loc=<5.7484 0 0>, rot=<0.1081 -0.0933 -0.0718 -0.9871>}
 base_bow_shoelace_lower_rt {parent=dyn_main_shoelaces_rt, loc=<0.2162 0.2582 0.1088>, rot=<-0.5846 -0.4111 0.5982 0.3625>}
 dyn_bow_shoelace_lower_rt {parent=base_bow_shoelace_lower_rt, loc=<0 0 0>, rot=<0 0 0 1>}
 calf_twist_02_r {parent=calf_r, loc=<19.1546 1.7943 0.2638>, rot=<0.0237 0.0095 -0.0035 -0.9997>}
 calf_twist_01_r {parent=calf_r, loc=<35.8483 1.9197 0.5794>, rot=<0.0237 0.0095 -0.0035 -0.9997>}
 deform_knee_in_base_r {parent=calf_r, loc=<-0.8031 5.0231 -0.2457>, rot=<0 0 0 -1>}
 deform_knee_in_r {parent=deform_knee_in_base_r, loc=<0 0 0>, rot=<0 0 0 1>}
 deform_knee_out_base_r {parent=calf_r, loc=<0 -4 0>, rot=<0 0 0 -1>}
 deform_knee_out_r {parent=deform_knee_out_base_r, loc=<0 0 0>, rot=<0 0 0 1>}
 thigh_twist_01_r {parent=thigh_r, loc=<27.952 0 0>, rot=<0.0013 0 0 -1>}
 base_charm_01_rt {parent=thigh_twist_01_r, loc=<-2.3696 -2.8405 12.4182>, rot=<-0.0506 0.1255 0.1316 -0.982>}
 dyn_charm_01_rt {parent=base_charm_01_rt, loc=<0 0 0>, rot=<0 0 0 1>}
 dyn_charm_02_rt {parent=dyn_charm_01_rt, loc=<1.711 -0.2435 0.3033>, rot=<0 0.0793 0 -0.9968>}
 deform_groin_base_r {parent=thigh_r, loc=<-0.4801 -5.6776 0.2486>, rot=<0 0 0 -1>}
 deform_groin_r {parent=deform_groin_base_r, loc=<0 0 0>, rot=<0 0 0 1>}
 deform_glute_base_l {parent=pelvis, loc=<-7.3466 -9.0213 -9.1245>, rot=<0.0172 0.031 -0.4477 -0.8935>}
 deform_glute_l {parent=deform_glute_base_l, loc=<0 0 0>, rot=<0 0 0 1>}
 deform_glute_base_r {parent=pelvis, loc=<-7.3464 -9.0213 9.1245>, rot=<0.031 -0.0172 0.8935 -0.4477>}
 deform_glute_r {parent=deform_glute_base_r, loc=<0 0 0>, rot=<0 0 0 1>}
 spine_03_null {parent=pelvis, type=Null, loc=<19.1247 -2.4625 0>, rot=<0 0 0.0726 -0.9974>}
 base_shirt_corr_cr {parent=pelvis, loc=<-9.2445 -11.7808 0>, rot=<-0.0179 -0.9997 0.0179 0.0003>}
 shirt_corr_cr {parent=base_shirt_corr_cr, loc=<0 0 0>, rot=<0 0 0 1>}
 attach {parent=root, loc=<0 0 -3.7397>, rot=<0 0 0.1217 -0.9926>}
 ik_hand_root {parent=root, loc=<0 0 -3.7397>, rot=<0 0 0.1203 -0.9927>}
 ik_hand_gun {parent=ik_hand_root, loc=<-48.1743 3.3545 104.4381>, rot=<0.6276 0.1309 0.4183 -0.6434>}
 ik_hand_r {parent=ik_hand_gun, loc=<0 0 0>, rot=<0 0 0 1>}
 ik_hand_l {parent=ik_hand_gun, loc=<59.3271 67.6997 34.3524>, rot=<-0.6158 -0.7027 -0.3565 0>}
 ik_foot_root {parent=root, loc=<0 0 -3.7397>, rot=<0 0 0.1231 -0.9924>}
 ik_foot_l {parent=ik_foot_root, loc=<16.1735 -3.5103 12.0357>, rot=<0 0 0 1>}
 ik_foot_r {parent=ik_foot_root, loc=<-12.5807 -10.7532 12.0357>, rot=<0 0 0 1>}
 root_fp {parent=root, loc=<0 0 -3.7397>, rot=<0 0 0.1217 -0.9926>}
 attach_fp {parent=root_fp, loc=<0 0 0>, rot=<0 0 0 1>}
 camera_root {parent=root_fp, loc=<0 0 0>, rot=<0 0 0 1>}
 camera {parent=camera_root, loc=<0 0 0>, rot=<0 0 0 1>}
 attach_emote_01 {parent=root, loc=<0 0 -3.7397>, rot=<0 0 0.1217 -0.9926>}
 attach_emote_02 {parent=root, loc=<0 0 -3.7397>, rot=<0 0 0.1217 -0.9926>}
 attach_emote_03 {parent=root, loc=<0 0 -3.7397>, rot=<0 0 0.1217 -0.9926>}
 warp_point_01 {parent=root, loc=<0 0 -3.7397>, rot=<0 0 0.1217 -0.9926>}
 warp_point_02 {parent=root, loc=<0 0 -3.7397>, rot=<0 0 0.1217 -0.9926>}
 warp_point_03 {parent=root, loc=<0 0 -3.7397>, rot=<0 0 0.1217 -0.9926>}
]
variables [IsCrouching {type=bool, default=False}]
