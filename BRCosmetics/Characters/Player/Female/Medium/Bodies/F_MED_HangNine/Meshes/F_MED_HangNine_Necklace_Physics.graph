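# Physics asset for the necklace on the HangNine medium female player character body.
# Capsules on the upper spine and both clavicles approximate the torso, alongside a three-capsule dynamic necklace chain hanging off spine_05.
# Necklace constraints zero twist and swing1, leaving a narrow swing2 range that widens at the last link, so the chain moves in a hinge-like way.
# Clavicle constraints to spine_05 are free in swing and twist.
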PhysicsAsset /BRCosmetics/Characters/Player/Female/Medium/Bodies/F_MED_HangNine/Meshes/F_MED_HangNine_Necklace_Physics
bodies {
 clavicle_l [{
   shapeType=Capsule
   length=7.6116
   radius=6.2346
   shapeName=clavicle_l_capsule
   center=<8.0847 -0.4946 -2.0522>
   rotation=<rot -81.0799 -142.3042 -46.6716>
  }]
 clavicle_r [{
   shapeType=Capsule
   length=7.6116
   radius=6.2346
   shapeName=clavicle_r_capsule
   center=<-8.0847 0.4946 2.0522>
   rotation=<rot -81.0799 -142.3042 133.3284>
  }]
 dyn_necklace_02_cr [{shapeType=Capsule, length=8.4663, radius=0.8677, shapeName=dyn_necklace_02_cr_capsule, center=<5 0 0>, rotation=<rot 0 0 90>}]
 dyn_necklace_03_cr [{shapeType=Capsule, length=3.5709, radius=1.4291, shapeName=dyn_necklace_03_cr_capsule, center=<5 0 0>, rotation=<rot 90 3 0>}]
 dyn_necklace_04_cr [{shapeType=Capsule, length=3.5709, radius=1.4291, shapeName=dyn_necklace_04_cr_capsule, center=<5 0 0>, rotation=<rot 90 3 0>}]
 spine_04 [{shapeType=Capsule, length=16.1232, radius=11.8751, center=<-1.7055 0.2215 0>, rotation=<rot 0 -90 0>}]
 spine_05 [{shapeType=Capsule, length=16.3456, radius=10.5031, center=<-2.2535 -2.7304 0.399>, rotation=<rot -1.4496 -90 2.7189>}]
}
constraints [
 {bone1Name=spine_05, bone2Name=spine_04, swing1LimitDegrees=45, swing2LimitDegrees=45, twistLimitDegrees=45}
 {bone1Name=dyn_necklace_02_cr, bone2Name=spine_05, swing1LimitDegrees=0, swing2LimitDegrees=5, twistLimitDegrees=0}
 {bone1Name=dyn_necklace_03_cr, bone2Name=dyn_necklace_02_cr, swing1LimitDegrees=0, swing2LimitDegrees=5, twistLimitDegrees=0}
 {bone1Name=dyn_necklace_04_cr, bone2Name=dyn_necklace_03_cr, swing1LimitDegrees=0, swing2LimitDegrees=15, twistLimitDegrees=0}
 {
  bone1Name=clavicle_l
  bone2Name=spine_05
  swing1LimitDegrees=45
  swing1Motion=Free
  swing2LimitDegrees=45
  swing2Motion=Free
  twistLimitDegrees=45
  twistMotion=Free
 }
 {
  bone1Name=clavicle_r
  bone2Name=spine_05
  swing1LimitDegrees=45
  swing1Motion=Free
  swing2LimitDegrees=45
  swing2Motion=Free
  twistLimitDegrees=45
  twistMotion=Free
 }
]
