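# Physics asset for the F_MED_HangNine female character body, rigging its jacket for secondary motion against pelvis, spine and thigh capsules.
# Seven jacket chains (center back, plus left and right front, side-front and side-back) each run from a small base body, free-rotating on spine_01, through three capsule segments.
# Segment joints lock twist and keep swing tight near the base, loosening toward the tips.
# Both front tips and the right thigh carry an extra capsule.
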PhysicsAsset /BRCosmetics/Characters/Player/Female/Medium/Bodies/F_MED_HangNine/Meshes/F_MED_HangNine_Physics
bodies {
 base_jacket_back_cr [{shapeType=Capsule, length=0.505, radius=0.505, shapeName=base_jacket_back_cr_capsule, rotation=<rot 0 -90 0>}]
 base_jacket_front_lf [{shapeType=Capsule, length=0.505, radius=0.505, shapeName=base_jacket_front_lf_capsule, rotation=<rot 0 -90 0>}]
 base_jacket_front_rt [{shapeType=Capsule, length=0.505, radius=0.505, shapeName=base_jacket_front_rt_capsule, rotation=<rot 0 -90 0>}]
 base_jacket_side_back_lf [{shapeType=Capsule, length=0.505, radius=0.505, shapeName=base_jacket_side_back_lf_capsule, rotation=<rot 0 -90 0>}]
 base_jacket_side_back_rt [{shapeType=Capsule, length=0.505, radius=0.505, shapeName=base_jacket_side_back_rt_capsule, rotation=<rot 0 -90 0>}]
 base_jacket_side_front_lf [{shapeType=Capsule, length=0.505, radius=0.505, shapeName=base_jacket_side_front_lf_capsule, rotation=<rot 0 -90 0>}]
 base_jacket_side_front_rt [{shapeType=Capsule, length=0.505, radius=0.505, shapeName=base_jacket_side_front_rt_capsule, rotation=<rot 0 -90 0>}]
 dyn_jacket_back_01_cr [{shapeType=Capsule, length=5, radius=1.5, shapeName=dyn_jacket_back_01_cr_capsule, center=<2 0 0>, rotation=<rot 90 0 0>}]
 dyn_jacket_back_02_cr [{shapeType=Capsule, length=5, radius=1.5, shapeName=dyn_jacket_back_02_cr_capsule, center=<2 0 0>, rotation=<rot 90 0 0>}]
 dyn_jacket_back_03_cr [{shapeType=Capsule, length=5, radius=1.5, shapeName=dyn_jacket_back_03_cr_capsule, center=<2 0 0>, rotation=<rot 90 0 0>}]
 dyn_jacket_front_01_lf [{shapeType=Capsule, length=4.149, radius=2.351, shapeName=dyn_jacket_front_01_lf_capsule, center=<2 0 0>, rotation=<rot 90 0 0>}]
 dyn_jacket_front_01_rt [{shapeType=Capsule, length=4.2409, radius=2.2591, shapeName=dyn_jacket_front_01_rt_capsule, center=<2 0 0>, rotation=<rot 90 0 0>}]
 dyn_jacket_front_02_lf [{shapeType=Capsule, length=4.2718, radius=2.2061, shapeName=dyn_jacket_front_02_lf_capsule, center=<2.2402 0 0>, rotation=<rot 90 0 0>}]
 dyn_jacket_front_02_rt [{shapeType=Capsule, length=4.2403, radius=2.2597, shapeName=dyn_jacket_front_02_rt_capsule, center=<2 0 0>, rotation=<rot 90 0 0>}]
 dyn_jacket_front_03_lf [
  {shapeType=Capsule, length=4.6205, radius=2.871, shapeName=dyn_jacket_front_03_lf_capsule, center=<2.2402 0 0>, rotation=<rot 90 0 0>}
  {
   shapeType=Capsule
   length=4.6205
   radius=2.871
   shapeName=dyn_jacket_front_03_lf_capsule
   center=<3.7147 0 -4.6602>
   rotation=<rot 0.0001 0 0>
  }
 ]
 dyn_jacket_front_03_rt [
  {shapeType=Capsule, length=5.4422, radius=2.2911, shapeName=dyn_jacket_front_03_rt_capsule, center=<2 0 0>, rotation=<rot 90 0 0>}
  {
   shapeType=Capsule
   length=5.2781
   radius=2.4553
   shapeName=dyn_jacket_front_03_rt_capsule
   center=<4.4047 0 3.8542>
   rotation=<rot -0.0001 -0.0001 0>
  }
 ]
 dyn_jacket_side_back_01_lf [{shapeType=Capsule, length=4.5506, radius=1.9494, shapeName=dyn_jacket_side_back_01_lf_capsule, center=<2 0 0>, rotation=<rot 90 0 0>}]
 dyn_jacket_side_back_01_rt [{shapeType=Capsule, length=4.5506, radius=1.9494, shapeName=dyn_jacket_side_back_01_rt_capsule, center=<2 0 0>, rotation=<rot 90 0 0>}]
 dyn_jacket_side_back_02_lf [{shapeType=Capsule, length=4.5506, radius=1.9494, shapeName=dyn_jacket_side_back_02_lf_capsule, center=<2 0 0>, rotation=<rot 90 0 0>}]
 dyn_jacket_side_back_02_rt [{shapeType=Capsule, length=4.3531, radius=2.1469, shapeName=dyn_jacket_side_back_02_rt_capsule, center=<2 0 0>, rotation=<rot 90 0 0>}]
 dyn_jacket_side_back_03_lf [{shapeType=Capsule, length=4.5506, radius=1.9494, shapeName=dyn_jacket_side_back_03_lf_capsule, center=<2 0 0>, rotation=<rot 90 0 0>}]
 dyn_jacket_side_back_03_rt [{shapeType=Capsule, length=4.5506, radius=1.9494, shapeName=dyn_jacket_side_back_03_rt_capsule, center=<2 0 0>, rotation=<rot 90 0 0>}]
 dyn_jacket_side_front_01_lf [{shapeType=Capsule, length=4.5506, radius=1.9494, shapeName=dyn_jacket_side_front_01_lf_capsule, center=<2 0 0>, rotation=<rot 90 0 0>}]
 dyn_jacket_side_front_01_rt [{shapeType=Capsule, length=4.5506, radius=1.9494, shapeName=dyn_jacket_side_front_01_rt_capsule, center=<4 0 0>, rotation=<rot 90 0 0>}]
 dyn_jacket_side_front_02_lf [{shapeType=Capsule, length=3.9646, radius=2.5354, shapeName=dyn_jacket_side_front_02_lf_capsule, center=<2 0 0>, rotation=<rot 90 0 0>}]
 dyn_jacket_side_front_02_rt [{shapeType=Capsule, length=4.5506, radius=1.9494, shapeName=dyn_jacket_side_front_02_rt_capsule, center=<4 0 0>, rotation=<rot 90 0 0>}]
 dyn_jacket_side_front_03_lf [{shapeType=Capsule, length=4.122, radius=2.378, shapeName=dyn_jacket_side_front_03_lf_capsule, center=<2 0 0>, rotation=<rot 90 0 0>}]
 dyn_jacket_side_front_03_rt [{shapeType=Capsule, length=3.7816, radius=2.269, shapeName=dyn_jacket_side_front_03_rt_capsule, center=<4 0 0>, rotation=<rot 90 0 0>}]
 pelvis [{shapeType=Capsule, length=10.5588, radius=10.4968, center=<0 -1.1627 0>, rotation=<rot 0 -90 0>}]
 spine_01 [{shapeType=Capsule, length=6.2335, radius=9.8906, shapeName=spine_01_capsule, center=<2.8938 -0.0582 0>, rotation=<rot 0 -90.0001 0>}]
 thigh_l [{shapeType=Capsule, length=22.3657, radius=8.668, center=<-21.2433 0.4161 0.2693>, rotation=<rot 0 -90.0851 -88.9694>}]
 thigh_r [
  {shapeType=Capsule, length=22.3657, radius=8.668, center=<21.2282 1.1454 -1.2368>, rotation=<rot 0 -90.0851 91.0306>}
  {shapeType=Capsule, length=24.224, radius=5.776, shapeName=thigh_r_capsule, center=<32.8319 -0.4453 7.3207>, rotation=<rot -85 180 180>}
 ]
}
constraints [
 {bone1Name=thigh_l, bone2Name=pelvis, swing1LimitDegrees=45, swing2LimitDegrees=45, twistLimitDegrees=45}
 {bone1Name=thigh_r, bone2Name=pelvis, swing1LimitDegrees=45, swing2LimitDegrees=45, twistLimitDegrees=45}
 {
  bone1Name=base_jacket_back_cr
  bone2Name=spine_01
  swing1LimitDegrees=45
  swing1Motion=Free
  swing2LimitDegrees=45
  swing2Motion=Free
  twistLimitDegrees=45
  twistMotion=Free
 }
 {
  bone1Name=dyn_jacket_back_01_cr
  bone2Name=base_jacket_back_cr
  swing1LimitDegrees=5
  swing2LimitDegrees=0
  twistLimitDegrees=45
  twistMotion=Locked
 }
 {
  bone1Name=dyn_jacket_back_02_cr
  bone2Name=dyn_jacket_back_01_cr
  swing1LimitDegrees=10
  swing2LimitDegrees=0
  twistLimitDegrees=45
  twistMotion=Locked
 }
 {
  bone1Name=dyn_jacket_back_03_cr
  bone2Name=dyn_jacket_back_02_cr
  swing1LimitDegrees=20
  swing2LimitDegrees=10
  twistLimitDegrees=45
  twistMotion=Locked
 }
 {
  bone1Name=base_jacket_front_lf
  bone2Name=spine_01
  swing1LimitDegrees=45
  swing1Motion=Free
  swing2LimitDegrees=45
  swing2Motion=Free
  twistLimitDegrees=45
  twistMotion=Free
 }
 {
  bone1Name=dyn_jacket_front_01_lf
  bone2Name=base_jacket_front_lf
  swing1LimitDegrees=5
  swing2LimitDegrees=0
  twistLimitDegrees=45
  twistMotion=Locked
 }
 {
  bone1Name=dyn_jacket_front_02_lf
  bone2Name=dyn_jacket_front_01_lf
  swing1LimitDegrees=20
  swing2LimitDegrees=10
  twistLimitDegrees=45
  twistMotion=Locked
 }
 {
  bone1Name=dyn_jacket_front_03_lf
  bone2Name=dyn_jacket_front_02_lf
  swing1LimitDegrees=20
  swing2LimitDegrees=10
  twistLimitDegrees=45
  twistMotion=Locked
 }
 {
  bone1Name=base_jacket_front_rt
  bone2Name=spine_01
  swing1LimitDegrees=45
  swing1Motion=Free
  swing2LimitDegrees=45
  swing2Motion=Free
  twistLimitDegrees=45
  twistMotion=Free
 }
 {
  bone1Name=dyn_jacket_front_01_rt
  bone2Name=base_jacket_front_rt
  swing1LimitDegrees=5
  swing2LimitDegrees=0
  twistLimitDegrees=45
  twistMotion=Locked
 }
 {
  bone1Name=dyn_jacket_front_02_rt
  bone2Name=dyn_jacket_front_01_rt
  swing1LimitDegrees=15
  swing2LimitDegrees=0
  twistLimitDegrees=45
  twistMotion=Locked
 }
 {
  bone1Name=dyn_jacket_front_03_rt
  bone2Name=dyn_jacket_front_02_rt
  swing1LimitDegrees=20
  swing2LimitDegrees=10
  twistLimitDegrees=45
  twistMotion=Locked
 }
 {
  bone1Name=base_jacket_side_back_lf
  bone2Name=spine_01
  swing1LimitDegrees=45
  swing1Motion=Free
  swing2LimitDegrees=45
  swing2Motion=Free
  twistLimitDegrees=45
  twistMotion=Free
 }
 {
  bone1Name=dyn_jacket_side_back_01_lf
  bone2Name=base_jacket_side_back_lf
  swing1LimitDegrees=5
  swing2LimitDegrees=0
  twistLimitDegrees=45
  twistMotion=Locked
 }
 {
  bone1Name=dyn_jacket_side_back_02_lf
  bone2Name=dyn_jacket_side_back_01_lf
  swing1LimitDegrees=5
  swing2LimitDegrees=0
  twistLimitDegrees=45
  twistMotion=Locked
 }
 {
  bone1Name=dyn_jacket_side_back_03_lf
  bone2Name=dyn_jacket_side_back_02_lf
  swing1LimitDegrees=20
  swing2LimitDegrees=10
  twistLimitDegrees=45
  twistMotion=Locked
 }
 {
  bone1Name=base_jacket_side_back_rt
  bone2Name=spine_01
  swing1LimitDegrees=45
  swing1Motion=Free
  swing2LimitDegrees=45
  swing2Motion=Free
  twistLimitDegrees=45
  twistMotion=Free
 }
 {
  bone1Name=dyn_jacket_side_back_01_rt
  bone2Name=base_jacket_side_back_rt
  swing1LimitDegrees=5
  swing2LimitDegrees=0
  twistLimitDegrees=45
  twistMotion=Locked
 }
 {
  bone1Name=dyn_jacket_side_back_02_rt
  bone2Name=dyn_jacket_side_back_01_rt
  swing1LimitDegrees=5
  swing2LimitDegrees=0
  twistLimitDegrees=45
  twistMotion=Locked
 }
 {
  bone1Name=dyn_jacket_side_back_03_rt
  bone2Name=dyn_jacket_side_back_02_rt
  swing1LimitDegrees=20
  swing2LimitDegrees=10
  twistLimitDegrees=45
  twistMotion=Locked
 }
 {
  bone1Name=base_jacket_side_front_lf
  bone2Name=spine_01
  swing1LimitDegrees=45
  swing1Motion=Free
  swing2LimitDegrees=45
  swing2Motion=Free
  twistLimitDegrees=45
  twistMotion=Free
 }
 {
  bone1Name=dyn_jacket_side_front_01_lf
  bone2Name=base_jacket_side_front_lf
  swing1LimitDegrees=5
  swing2LimitDegrees=0
  twistLimitDegrees=45
  twistMotion=Locked
 }
 {
  bone1Name=dyn_jacket_side_front_02_lf
  bone2Name=dyn_jacket_side_front_01_lf
  swing1LimitDegrees=5
  swing2LimitDegrees=0
  twistLimitDegrees=45
  twistMotion=Locked
 }
 {
  bone1Name=dyn_jacket_side_front_03_lf
  bone2Name=dyn_jacket_side_front_02_lf
  swing1LimitDegrees=20
  swing2LimitDegrees=10
  twistLimitDegrees=45
  twistMotion=Locked
 }
 {
  bone1Name=base_jacket_side_front_rt
  bone2Name=spine_01
  swing1LimitDegrees=45
  swing1Motion=Free
  swing2LimitDegrees=45
  swing2Motion=Free
  twistLimitDegrees=45
  twistMotion=Free
 }
 {
  bone1Name=dyn_jacket_side_front_01_rt
  bone2Name=base_jacket_side_front_rt
  swing1LimitDegrees=10
  swing2LimitDegrees=0
  twistLimitDegrees=45
  twistMotion=Locked
 }
 {
  bone1Name=dyn_jacket_side_front_02_rt
  bone2Name=dyn_jacket_side_front_01_rt
  swing1LimitDegrees=15
  swing2LimitDegrees=0
  twistLimitDegrees=45
  twistMotion=Locked
 }
 {
  bone1Name=dyn_jacket_side_front_03_rt
  bone2Name=dyn_jacket_side_front_02_rt
  swing1LimitDegrees=20
  swing2LimitDegrees=10
  twistLimitDegrees=45
  twistMotion=Locked
 }
 {
  bone1Name=spine_01
  bone2Name=pelvis
  swing1LimitDegrees=45
  swing1Motion=Free
  swing2LimitDegrees=45
  swing2Motion=Free
  twistLimitDegrees=45
  twistMotion=Free
 }
]
