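# Control Rig for the HangNine female character's face-accessory mesh, posing back hair and ponytail bones procedurally around head, neck and shoulder motion.
# After a shared deform-rig function, the "pony tail" function applies aim and parent constraints, then layers weighted corrective offsets.
# Weights combine swing-twist delta rotations, spherical pose readers and pelvis height with crouch, targeting and RBAN flags.
# It also outputs an "aiming up pistol" curve from the inverted pistol-aim weight.
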ControlRigBlueprint /BRCosmetics/Characters/Player/Female/Medium/Bodies/F_MED_HangNine/Meshes/Parts/F_MED_HangNine_FaceAcc_CtrlRig
graphs [
 Multiply_6_ContainedGraph {nodes [
   n1 = RigVMFunctionEntryNode()
   RigVMFunctionReturnNode(Result=&RigVMFunction_MathDoubleMul(B=@n1.C, A=&RigVMFunction_MathDoubleMul(A=@n1.A, B=@n1.B).Result).Result)
  ]}
 Multiply_7_ContainedGraph {nodes [
   n1 = RigVMFunctionEntryNode()
   RigVMFunctionReturnNode(Result=&RigVMFunction_MathDoubleMul(B=@n1.C, A=&RigVMFunction_MathDoubleMul(A=@n1.A, B=@n1.B).Result).Result)
  ]}
 Multiply_7_ContainedGraph {nodes [
   n1 = RigVMFunctionEntryNode()
   RigVMFunctionReturnNode(Result=&RigVMFunction_MathDoubleMul(B=@n1.C, A=&RigVMFunction_MathDoubleMul(A=@n1.A, B=@n1.B).Result).Result)
  ]}
 New Function_ContainedGraph {nodes [
   RigVMFunctionEntryNode()
   > RigVMFunction_Sequence = RigVMAggregateNode()
   &RigUnit_AimConstraintLocalSpaceOffset(
    AdvancedSettings=(DebugSettings=(bEnabled=false,Scale=10,WorldOffset=(Rotation=(X=0,Y=0,Z=0,W=1),Translation=(X=0,Y=0,Z=0),Scale3D=(X=1,Y=1,Z=1))),RotationOrderForFilter=XZY)
    Child=(Type=Bone,Name="base_hair_back_01_cr")
    Filter=(bX=True,bY=True,bZ=True)
    Parents=((Item=(Type=Bone,Name="dyn_hair_back_aim"),Weight=1))
    UpAxis=(X=0,Y=-1,Z=0)
    Weight=1
    WorldUp=(Target=(X=0,Y=-1,Z=0),Kind=Location,Space=(Type=Bone,Name="dyn_hair_back_aim"))
    ExecutePin=@RigVMFunction_Sequence.A
   )
   > &RigUnit_AimConstraintLocalSpaceOffset(
    AdvancedSettings=(DebugSettings=(bEnabled=false,Scale=10,WorldOffset=(Rotation=(X=0,Y=0,Z=0,W=1),Translation=(X=0,Y=0,Z=0),Scale3D=(X=1,Y=1,Z=1))),RotationOrderForFilter=XZY)
    Child=(Type=Bone,Name="base_tail_hair_03_lf")
    Filter=(bX=True,bY=True,bZ=True)
    Parents=((Item=(Type=Bone,Name="dyn_tail_hair_for_aim"),Weight=1))
    UpAxis=(X=0,Y=1,Z=0)
    Weight=1
    WorldUp=(Target=(X=0,Y=0,Z=1),Kind=Direction,Space=(Type=Bone,Name="dyn_tail_hair_for_aim"))
   )
   > &RigUnit_ParentConstraint(
    AdvancedSettings=(RotationOrderForFilter=XZY)
    Child=(Type=Bone,Name="base_tail_hair_03_lf")
    Parents="((Item=(Type=Bone,Name=\"head\"),Weight=0.0183),(Item=(Type=Bone,Name=\"neck_01\"),Weight=0.0819),(Item=(Type=Bone,Name=\"neck_02\"),Weight=0.0891),(Item=(Type=Bone,Name=\"spine_03\"),Weight=0.1053),(Item=(Type=Bone,Name=\"clavicle_l\"),Weight=0.1324),(Item=(Type=Bo ...(truncated)"
    Weight=1
   )
   > &RigUnit_ParentConstraint(
    AdvancedSettings=(RotationOrderForFilter=XZY)
    Child=(Type=Bone,Name="dyn_tail_hair_03_lf")
    Parents="((Item=(Type=Bone,Name=\"head\"),Weight=0.0183),(Item=(Type=Bone,Name=\"neck_01\"),Weight=0.0819),(Item=(Type=Bone,Name=\"neck_02\"),Weight=0.0891),(Item=(Type=Bone,Name=\"spine_03\"),Weight=0.1053),(Item=(Type=Bone,Name=\"clavicle_l\"),Weight=0.1324),(Item=(Type=Bo ...(truncated)"
    Weight=1
   )
   > &RigUnit_ParentConstraint(
    AdvancedSettings=(RotationOrderForFilter=XZY)
    Child=(Type=Bone,Name="dyn_tail_hair_04_lf")
    Parents="((Item=(Type=Bone,Name=\"head\"),Weight=0.0238),(Item=(Type=Bone,Name=\"dyn_neck_shirt_lf\"),Weight=0.0411),(Item=(Type=Bone,Name=\"neck_02\"),Weight=0.1532),(Item=(Type=Bone,Name=\"neck_01\"),Weight=0.1642),(Item=(Type=Bone,Name=\"clavicle_l\"),Weight=0.1701),(Item ...(truncated)"
    Weight=1
   )
   DeltaRotation = RigVMFunctionReferenceNode(
    Rotation Order=ZYX
    child=head
    debugRotation=false
    max=-30
    min=-10
    onlyTwist=false
    parent=neck_02
    parentType=Bone
    premult=1
    primAxis=2
    twistAxis=(X=1,Y=0,Z=0)
    ExecuteContext=@RigVMFunction_Sequence.C
   )
   > &RigUnit_ModifyTransforms(
    ItemToModify="((Item=(Type=Bone,Name=\"base_hair_back_01_cr\"),Transform=(Rotation=(X=-0.0359,Y=0.0027,Z=0.0126,W=-0.9993),Translation=(X=0,Y=0,Z=0),Scale3D=(X=1,Y=1,Z=1))),(Item=(Type=Bone,Name=\"dyn_tail_hair_04_lf\"),Transform=(Rotation=(X=0.0275,Y=-0.067,Z=-0.0967,W=0.9 ...(truncated)"
    WeightMinimum=0
    Weight=@DeltaRotation.outputWeight
   )
   > DeltaRotation_1 = RigVMFunctionReferenceNode(
    Rotation Order=ZYX
    child=head
    debugRotation=false
    max=-50
    min=-30
    onlyTwist=false
    parent=neck_02
    parentType=Bone
    premult=1
    primAxis=2
    twistAxis=(X=1,Y=0,Z=0)
   )
   Cast = &RigVMFunction_MathBoolToFloat(Value=RigVMVariableNode() {variable=is Crouching}.Value)
   > RigVMVariableNode(Value=&RigVMFunction_MathDoubleMul(B=@Cast.Result, A=@DeltaRotation_1.outputWeight).Result) {variable=head crouching}
   DeltaRotation_3 = RigVMFunctionReferenceNode(
    Rotation Order=ZYX
    child=head
    debugRotation=false
    max=50
    min=-10
    onlyTwist=false
    parent=spine_05
    parentType=Bone
    premult=1
    primAxis=2
    twistAxis=(X=1,Y=0,Z=0)
    ExecuteContext=@RigVMFunction_Sequence.E
   )
   > n1 = &RigUnit_SphericalPoseReader(
    ActiveRegionScaleFactors=(PositiveWidth=1,NegativeWidth=1,PositiveHeight=1,NegativeHeight=1)
    ActiveRegionSize=0.1
    Debug=(bDrawDebug=true,bDraw2D=False,bDrawLocalAxes=False,DebugScale=25,DebugSegments=20,DebugThickness=0.25)
    DriverAxis=(X=1,Y=0,Z=0)
    DriverItem=(Type=Bone,Name="upperarm_l")
    FalloffRegionScaleFactors=(PositiveWidth=1,NegativeWidth=1,PositiveHeight=1,NegativeHeight=1)
    FalloffSize=0.4
    FlipHeightScaling=False
    FlipWidthScaling=False
    OptionalParentItem=(Type=Bone,Name="None")
    RotationOffset=(X=130,Y=-20,Z=-90)
   )
   n2 = &RigVMFunction_AlphaInterp(
    InterpSpeedDecreasing=10
    InterpSpeedIncreasing=10
    Scale=1
    bClampResult=False
    Value=&RigVMFunction_MathBoolToFloat(Value=RigVMVariableNode() {variable=is Targeting}.Value).Result
   )
   n3 = &RigVMFunction_AnimEvalRichCurve(
    Curve="(EditorCurveData=(Keys=((InterpMode=RCIM_Cubic,TangentMode=RCTM_Auto,TangentWeightMode=RCTWM_WeightedNone,Time=0,Value=0,ArriveTangent=0,ArriveTangentWeight=0,LeaveTangent=0,LeaveTangentWeight=0),(InterpMode=RCIM_Cubic,TangentMode=RCTM_Auto,TangentWeightMo ...(truncated)"
    SourceMinimum=0
    TargetMinimum=0
    Value=@n2.Result
   )
   Multiply_5 = RigVMAggregateNode(B=@DeltaRotation_3.outputWeight, C=@n3.Result, A=@n1.OutputParam)
   > &RigUnit_ModifyTransforms(
    ItemToModify="((Item=(Type=Bone,Name=\"dyn_hair_back_01_cr\"),Transform=(Rotation=(X=0.1656,Y=-0.0063,Z=0.1767,W=-0.9702),Translation=(X=0,Y=0,Z=0),Scale3D=(X=1,Y=1,Z=1))),(Item=(Type=Bone,Name=\"dyn_tail_hair_02_lf\"),Transform=(Rotation=(X=0.3134,Y=0.0038,Z=0.0335,W=0.949 ...(truncated)"
    WeightMinimum=0
    Weight=@Multiply_5.Result
   )
   n4 = RigVMVariableNode() {variable=RBAN enabled}
   > &RigUnit_ModifyTransforms(
    ItemToModify="((Item=(Type=Bone,Name=\"dyn_tail_hair_ik_07_lf\"),Transform=(Rotation=(X=-0.0088,Y=0.0683,Z=-0.1276,W=-0.9894),Translation=(X=-0.2647,Y=2.5543,Z=0.2303),Scale3D=(X=1,Y=1,Z=1))),(Item=(Type=Bone,Name=\"dyn_tail_hair_ik_06_lf\"),Transform=(Rotation=(X=0,Y=0,Z=0 ...(truncated)"
    WeightMinimum=0
    Weight=&RigVMFunction_MathDoubleMul(
     B=&RigVMFunction_MathBoolToFloat(Value=@n4.Value).Result
     A=&RigVMFunction_MathDoubleRemap(SourceMaximum=1, SourceMinimum=0, TargetMaximum=1, TargetMinimum=0, Value=@Multiply_5.Result).Result
    ).Result
   )
   > RigVMFunctionReturnNode()
   Remap = &RigVMFunction_MathDoubleRemap(
    SourceMaximum=45
    SourceMinimum=89
    TargetMaximum=1
    TargetMinimum=0
    Value=&RigUnit_GetRelativeTransformForItem(Child=(Type=Bone,Name="pelvis"), Parent=(Type=Bone,Name="root"), bChildInitial=False, bParentInitial=False).RelativeTransform.Translation.Z
   )
   Multiply = &RigVMFunction_MathDoubleMul(A=@Remap.Result, B=RigVMVariableNode() {variable=head crouching}.Value)
   AlphaInterp = &RigVMFunction_AlphaInterp(InterpSpeedDecreasing=10, InterpSpeedIncreasing=10, Scale=1, bClampResult=False, Value=@Multiply.Result)
   &RigUnit_ModifyTransforms(
    ItemToModify="((Item=(Type=Bone,Name=\"base_hair_back_01_cr\"),Transform=(Rotation=(X=0,Y=0,Z=0,W=1),Translation=(X=0,Y=-1,Z=0),Scale3D=(X=1,Y=1,Z=1))),(Item=(Type=Bone,Name=\"dyn_tail_hair_04_lf\"),Transform=(Rotation=(X=0.0038,Y=0.005,Z=-0.0041,W=1),Translation=(X=-0.2359 ...(truncated)"
    WeightMinimum=0
    Weight=@AlphaInterp.Result
    ExecutePin=@RigVMFunction_Sequence.D
   )
   n5 = &RigVMFunction_AnimEvalRichCurve(
    Curve=(EditorCurveData=(Keys=((),(Time=0.5,Value=1),(Time=0.8))))
    SourceMaximum=2
    SourceMinimum=0
    TargetMinimum=0
    Value=&RigVMFunction_MathDoubleRemap(
     SourceMaximum=75
     SourceMinimum=45
     TargetMaximum=1
     TargetMinimum=0
     bClamp=false
     Value=&RigUnit_GetRelativeTransformForItem(Child=(Type=Bone,Name="pelvis"), Parent=(Type=Bone,Name="root"), bChildInitial=False, bParentInitial=False).RelativeTransform.Translation.Z
    ).Result
   )
   n6 = &RigVMFunction_MathDoubleMul(B=&RigVMFunction_MathBoolToFloat(Value=RigVMVariableNode() {variable=is Crouching}.Value).Result, A=@n5.Result)
   > RigVMVariableNode(Value=@n6.Result) {variable=crouching move}
   > &RigUnit_ModifyTransforms(
    ItemToModify="((Item=(Type=Bone,Name=\"base_hair_back_01_cr\"),Transform=(Rotation=(X=-0.0712,Y=-0.1268,Z=0.1141,W=0.9828),Translation=(X=1,Y=0,Z=0),Scale3D=(X=1,Y=1,Z=1))),(Item=(Type=Bone,Name=\"base_tail_hair_03_lf\"),Transform=(Rotation=(X=0,Y=0,Z=0,W=1),Translation=(X= ...(truncated)"
    WeightMinimum=0
    Weight=@n6.Result
   )
   > n7 = &RigUnit_SphericalPoseReader(
    ActiveRegionScaleFactors=(PositiveWidth=1,NegativeWidth=1,PositiveHeight=1,NegativeHeight=1)
    ActiveRegionSize=0.05
    Debug=(bDrawDebug=false,bDraw2D=False,bDrawLocalAxes=False,DebugScale=25,DebugSegments=20,DebugThickness=0.25)
    DriverAxis=(X=1,Y=0,Z=0)
    DriverItem=(Type=Bone,Name="clavicle_l")
    FalloffRegionScaleFactors=(PositiveWidth=1,NegativeWidth=1,PositiveHeight=1,NegativeHeight=1)
    FalloffSize=0.12
    FlipHeightScaling=False
    FlipWidthScaling=False
    OptionalParentItem=(Type=Bone,Name="None")
    RotationOffset=(X=0,Y=-55,Z=5)
   )
   > RigVMVariableNode(Value=@n7.OutputParam) {variable=clavicle up for targeting}
   n8 = &RigVMFunction_AlphaInterp(InterpSpeedDecreasing=8, InterpSpeedIncreasing=8, Scale=1, bClampResult=false, bMapRange=false, Value=@n7.OutputParam)
   RigVMFunction_AnimEvalRichCurve = &RigVMFunction_AnimEvalRichCurve(
    Curve="(EditorCurveData=(Keys=((InterpMode=RCIM_Linear,TangentMode=RCTM_Auto,TangentWeightMode=RCTWM_WeightedNone,Time=0,Value=0,ArriveTangent=-0.052,ArriveTangentWeight=0,LeaveTangent=0,LeaveTangentWeight=0),(InterpMode=RCIM_Linear,TangentMode=RCTM_Auto,TangentW ...(truncated)"
    SourceMinimum=0
    TargetMinimum=0
    Value=@n8.Result
   )
   > &RigUnit_ModifyTransforms(
    ItemToModify="((Item=(Type=Bone,Name=\"base_hair_back_01_cr\"),Transform=(Rotation=(X=0,Y=0,Z=0,W=1),Translation=(X=0,Y=-0.5,Z=0),Scale3D=(X=1,Y=1,Z=1))),(Item=(Type=Bone,Name=\"base_tail_hair_03_lf\"),Transform=(Rotation=(X=-0.0017,Y=-0.048,Z=0.0352,W=-0.9982),Translation= ...(truncated)"
    WeightMinimum=0
    Weight=@RigVMFunction_AnimEvalRichCurve.Result
   )
   > n9 = &RigUnit_SphericalPoseReader(
    ActiveRegionScaleFactors=(PositiveWidth=1,NegativeWidth=0.5,PositiveHeight=0.75,NegativeHeight=0.75)
    ActiveRegionSize=0.1
    Debug=(bDrawDebug=false,bDraw2D=False,bDrawLocalAxes=False,DebugScale=25,DebugSegments=20,DebugThickness=0.25)
    DriverAxis=(X=0,Y=1,Z=0)
    DriverItem=(Type=Bone,Name="head")
    FalloffRegionScaleFactors=(PositiveWidth=1,NegativeWidth=0.5,PositiveHeight=1,NegativeHeight=1)
    FalloffSize=0.2
    FlipHeightScaling=False
    FlipWidthScaling=False
    OptionalParentItem=(Type=Bone,Name="None")
    RotationOffset=(X=90,Y=0,Z=-30)
   )
   > n10 = &RigUnit_SphericalPoseReader(
    ActiveRegionScaleFactors=(PositiveWidth=1,NegativeWidth=1,PositiveHeight=1,NegativeHeight=1)
    ActiveRegionSize=0.05
    Debug=(bDrawDebug=false,bDraw2D=False,bDrawLocalAxes=False,DebugScale=25,DebugSegments=20,DebugThickness=0.25)
    DriverAxis=(X=1,Y=0,Z=0)
    DriverItem=(Type=Bone,Name="upperarm_l")
    FalloffRegionScaleFactors=(PositiveWidth=1,NegativeWidth=1,PositiveHeight=1,NegativeHeight=1)
    FalloffSize=0.15
    FlipHeightScaling=False
    FlipWidthScaling=False
    OptionalParentItem=(Type=Bone,Name="None")
    RotationOffset=(X=60,Y=-35,Z=-15)
   )
   Multiply_6 = RigVMAggregateNode(
    C=&RigVMFunction_MathDoubleRemap(SourceMaximum=0.8, SourceMinimum=0, TargetMaximum=0, TargetMinimum=1, Value=RigVMVariableNode() {variable=head crouching}.Value).Result
    A=@n9.OutputParam
    B=@n10.OutputParam
   )
   Subtract = &RigVMFunction_MathDoubleSub(B=1, A=@Multiply_6.Result)
   Absolute = &RigVMFunction_MathDoubleAbs(Value=@Subtract.Result)
   > /Script/ControlRig.RigUnit_SetCurveValue(Curve=aiming up pistol, Value=@Absolute.Result) {is=struct RigVMUnitNode}
   > RigVMVariableNode(Value=@Multiply_6.Result) {variable=aiming pistol}
   > &RigUnit_ModifyTransforms(
    ItemToModify="((Item=(Type=Bone,Name=\"dyn_tail_hair_02_lf\"),Transform=(Rotation=(X=-0.1582,Y=0.1013,Z=-0.1108,W=0.9759),Translation=(X=-0.5937,Y=-1.0414,Z=-1.4315),Scale3D=(X=1,Y=1,Z=1))),(Item=(Type=Bone,Name=\"base_tail_hair_03_lf\"),Transform=(Rotation=(X=-0.1196,Y=0.1 ...(truncated)"
    WeightMinimum=0
    Weight=@Multiply_6.Result
   )
   > n11 = &RigUnit_SphericalPoseReader(
    ActiveRegionScaleFactors=(PositiveWidth=1,NegativeWidth=1,PositiveHeight=1,NegativeHeight=1)
    ActiveRegionSize=0.05
    Debug=(bDrawDebug=false,bDraw2D=false,bDrawLocalAxes=false,DebugScale=25,DebugSegments=20,DebugThickness=0.25)
    DriverAxis=(X=1,Y=0,Z=0)
    DriverItem=(Type=Bone,Name="clavicle_l")
    FalloffRegionScaleFactors=(PositiveWidth=1,NegativeWidth=1,PositiveHeight=1,NegativeHeight=1)
    FalloffSize=0.15
    FlipHeightScaling=False
    FlipWidthScaling=False
    OptionalParentItem=(Type=Bone,Name="None")
    RotationOffset=(X=0,Y=-30,Z=15)
   )
   > &RigUnit_ModifyTransforms(
    ItemToModify="((Item=(Type=Bone,Name=\"base_hair_back_01_cr\"),Transform=(Rotation=(X=-0.0359,Y=-0.1908,Z=0.048,W=0.9798),Translation=(X=0.8807,Y=-0.4053,Z=0.2918),Scale3D=(X=0.937,Y=1,Z=1))),(Item=(Type=Bone,Name=\"dyn_tail_hair_02_lf\"),Transform=(Rotation=(X=-0.1834,Y=-0 ...(truncated)"
    WeightMinimum=0
    Weight=&RigVMFunction_MathDoubleMul(A=@n11.OutputParam, B=RigVMVariableNode() {variable=head crouching}.Value).Result
   )
   > DeltaRotation_2 = RigVMFunctionReferenceNode(
    Rotation Order=ZYX
    child=head
    debugRotation=false
    max=-55
    min=-36
    onlyTwist=false
    parent=neck_01
    parentType=Bone
    premult=1
    primAxis=2
    twistAxis=(X=1,Y=0,Z=0)
   )
   Multiply_4 = RigVMAggregateNode(
    C=&RigVMFunction_MathBoolToFloat(Value=RigVMVariableNode() {variable=is Targeting}.Value).Result
    A=@DeltaRotation_2.outputWeight
    B=RigVMVariableNode() {variable=crouching move}.Value
   )
   n12 = &RigVMFunction_MathDoubleRemap(
    SourceMaximum=0.7
    SourceMinimum=0
    TargetMinimum=0
    Value=@Multiply_4.Result
    TargetMaximum=&RigVMFunction_MathDoubleRemap(SourceMaximum=0.2, SourceMinimum=0.7, TargetMaximum=0.4, TargetMinimum=1, Value=@Multiply_4.Result).Result
   )
   > &RigUnit_ModifyTransforms(
    ItemToModify="((Item=(Type=Bone,Name=\"dyn_hair_back_01_cr\"),Transform=(Rotation=(X=0.0678,Y=-0.0604,Z=0.0362,W=0.9952),Translation=(X=-4.0281,Y=0.1429,Z=1.5304),Scale3D=(X=1,Y=1,Z=1))),(Item=(Type=Bone,Name=\"dyn_hair_back_02_cr\"),Transform=(Rotation=(X=0,Y=0,Z=0,W=1),Tr ...(truncated)"
    WeightMinimum=0
    Weight=@n12.Result
   )
   > n13 = &RigUnit_SphericalPoseReader(
    ActiveRegionScaleFactors=(PositiveWidth=1,NegativeWidth=1,PositiveHeight=1,NegativeHeight=1)
    ActiveRegionSize=0.03
    Debug=(bDrawDebug=false,bDraw2D=False,bDrawLocalAxes=False,DebugScale=25,DebugSegments=20,DebugThickness=0.25)
    DriverAxis=(X=1,Y=0,Z=0)
    DriverItem=(Type=Bone,Name="clavicle_l")
    FalloffRegionScaleFactors=(PositiveWidth=1,NegativeWidth=1,PositiveHeight=1,NegativeHeight=1)
    FalloffSize=0.04
    FlipHeightScaling=False
    FlipWidthScaling=False
    OptionalParentItem=(Type=Bone,Name="None")
    RotationOffset=(X=-20,Y=-35,Z=4)
   )
   > &RigUnit_ModifyTransforms(
    ItemToModify="((Item=(Type=Bone,Name=\"dyn_tail_hair_02_lf\"),Transform=(Rotation=(X=0.0809,Y=0.0221,Z=-0.0699,W=0.994),Translation=(X=0,Y=0,Z=0),Scale3D=(X=1,Y=1,Z=1))),(Item=(Type=Bone,Name=\"base_tail_hair_03_lf\"),Transform=(Rotation=(X=-0.1384,Y=0.0117,Z=0.0838,W=-0.98 ...(truncated)"
    WeightMinimum=0
    Weight=&RigVMFunction_MathDoubleMul(B=@n12.Result, A=@n13.OutputParam).Result
   )
  ]}
 New Function_ContainedGraph {nodes [
   n1 = RigVMFunctionEntryNode()
   RigVMFunction_MathBoolToInteger = &RigVMFunction_MathBoolToInteger(Value=@n1.onlyTwist)
   RigVMFunction_MathQuaternionSwingTwist = &RigVMFunction_MathQuaternionSwingTwist(
    twistAxis=@n1.twistAxis
    Input=&RigUnit_GetRelativeTransformForItem(
     bChildInitial=False
     bParentInitial=False
     child=(Type=Bone,Name="None")
     parent=(Type=Bone,Name="None")
     child.Name=@n1.child
     parent.Name=@n1.parent
     parent.Type=@n1.parentType
    ).RelativeTransform.Rotation
   )
   RigVMFunction_MathQuaternionToEuler = &RigVMFunction_MathQuaternionToEuler(
    Value=RigVMDispatchNode(
     Values=((),())
     Index=@RigVMFunction_MathBoolToInteger.Result
     Values.0=@RigVMFunction_MathQuaternionSwingTwist.Swing
     Values.1=@RigVMFunction_MathQuaternionSwingTwist.Twist
    ) {template="DISPATCH_RigVMDispatch_SelectInt32(in Index,in Values,out Result)"}.Result
    RotationOrder=@n1.Rotation Order
   )
   > DISPATCH_RigVMDispatch_Print = RigVMDispatchNode(ScreenColor=(R=1,G=1,B=1,A=1), ScreenDuration=1.5, Enabled=@n1.debugRotation, Value=@RigVMFunction_MathQuaternionToEuler.Result) {template="DISPATCH_RigVMDispatch_Print(in Prefix,in Value,in Enabled,in ScreenDuration,in ScreenColor)"}
   Clamp = &RigVMFunction_MathIntClamp(Value=@n1.primAxis)
   DISPATCH_RigVMDispatch_SelectInt32 = RigVMDispatchNode(
    Values=(0,0,0)
    Index=@Clamp.Result
    Values.0=@RigVMFunction_MathQuaternionToEuler.Result.X
    Values.1=@RigVMFunction_MathQuaternionToEuler.Result.Y
    Values.2=@RigVMFunction_MathQuaternionToEuler.Result.Z
   ) {template="DISPATCH_RigVMDispatch_SelectInt32(in Index,in Values,out Result)"}
   Remap = &RigVMFunction_MathDoubleRemap(
    TargetMaximum=1
    TargetMinimum=0
    bClamp=False
    Value=@DISPATCH_RigVMDispatch_SelectInt32.Result
    SourceMaximum=@n1.max
    SourceMinimum=@n1.min
   )
   Multiply = &RigVMFunction_MathDoubleMul(A=&RigVMFunction_MathDoubleClamp(Value=@Remap.Result).Result, B=@n1.premult)
   > RigVMFunctionReturnNode(outputWeight=@Multiply.Result)
  ]}
 RigVMFunctionLibrary {nodes [
   DeltaRotation = RigVMCollapseNode(ExecuteContext=(), Rotation Order=ZYX, premult=1, twistAxis=(X=1,Y=0,Z=0))
   pony tail = RigVMCollapseNode(ExecuteContext=())
  ]}
 RigVMFunction_Sequence_1_ContainedGraph {nodes [
   n1 = RigVMFunctionEntryNode()
   RigVMFunction_Sequence = &RigVMFunction_Sequence(ExecuteContext=@n1.ExecuteContext)
   n2 = &RigVMFunction_Sequence(ExecuteContext=@RigVMFunction_Sequence.B)
   n3 = &RigVMFunction_Sequence(ExecuteContext=@n2.B)
   RigVMFunctionReturnNode(A=@RigVMFunction_Sequence.A, D=@n3.A, E=@n3.B, C=@n2.A)
  ]}
 RigVMModel {nodes [
   RigUnit_BeginExecution = &RigUnit_BeginExecution()
   FN_DeformRig_FNC = RigVMFunctionReferenceNode(
    DeformRig_ Curve="(deformRig_l_upperarm_up_alpha_72_#1=False,deformRig_r_upperarm_up_alpha_73_#2=False,deformRig_l_upperarm_fr_alpha_74_#3=False,deformRig_r_upperarm_fr_alpha_75_CB51B8 ...(truncated)"
    DeformRig_Bool="(upperarm_1_#4=True,elbow_3_#5=True,wrist_7_#6=True,groin_8_#7=True,glute_17_#8=True,knee_9_40C408324CA261 ...(truncated)"
    DeformRig_Debug=(upperarm_debug_18_#4=False,wrist_debug_20_#6=False,lat_debug_26_#9=False,groinGlute_debug_27_#7=False)
    DeformRig_Value="(upperarm_up_18_#10=(Rotation=(X=0.0,Y=-0.0,Z=0.0,W=1.0),Translation=(X=0.0,Y=0.0,Z=8.0),Scale3D=(X=0.70,Y=1.0,Z=1.0)),upperarm_fwd_fr_19_#11=(Rotation=(X=0.0,Y=-0.0,Z=0.0,W=1.0),Translation=(X=0.0, ...(truncated)"
    ExecuteContext=@RigUnit_BeginExecution.ExecutePin
   )
   > pony tail = RigVMFunctionReferenceNode()
   &RigUnit_PrepareForExecution()
   addFKCtrl = RigVMFunctionReferenceNode(
    Custom_FKCtrl=()
    ExecuteContext=()
    Options_FKCtrl="(pelvis_1_#12=false,spine_3_#13=false,clavicle_5_#14=true,arm_7_#15=false,leg_9_#16=false,head_11_46DC15EF465CD ...(truncated)"
    Shape_Settings="(pelvis_2_#17=(bVisible=True,Name=\"Hexagon_Thin\",Color=(R=0.0117,G=0.8281,B=1,A=1),Transform=(Rotation=(X=0,Y=-0.7071,Z=0,W=0.7071),Translation=(X=2,Y=0,Z=0),Scale3D=(X=5,Y=5,Z=7))),spine_4_#18=(bVi ...(truncated)"
   )
   setFKCtrl = RigVMFunctionReferenceNode(ExecuteContext=(), setFK_Ctrl=false)
  ]}
]
hierarchy [
 root {loc=<0 0 0>, rot=<0 0 0 1>}
 pelvis {parent=root, loc=<0 -0.5301 98.0571>, rot=<0.0127 0.707 -0.0127 -0.707>}
 spine_01 {parent=pelvis, loc=<6.1208 -0.9328 0>, rot=<0 0 0.034 -0.9994>}
 spine_02 {parent=spine_01, loc=<6.5997 -0.2095 0>, rot=<0 0 0.0238 -0.9997>}
 spine_03 {parent=spine_02, loc=<6.4914 -0.1237 0>, rot=<0 0 0.0149 -0.9999>}
 spine_04 {parent=spine_03, loc=<6.9598 0.0794 0>, rot=<0 0 -0.0132 -0.9999>}
 spine_05 {parent=spine_04, loc=<6.9987 0 0>, rot=<0 0 -0.0015 -1>}
 neck_01 {parent=spine_05, loc=<11.3138 0.1529 0>, rot=<0 0 -0.2003 -0.9797>}
 neck_02 {parent=neck_01, loc=<4.5627 0 0>, rot=<0 0 0.0504 -0.9987>}
 head {parent=neck_02, loc=<5.3914 0 0>, rot=<0 0 0.115 -0.9934>}
 base_hair_01_cr {parent=head, loc=<6.3434 9.4089 -0.8199>, rot=<0.2248 0.0139 0.9743 0.0024>}
 dyn_hair_01_cr {parent=base_hair_01_cr, loc=<0 0 0>, rot=<0 0 0 1>}
 base_hair_01_lf {parent=head, loc=<6.6848 3.7173 -6.2133>, rot=<-0.0005 0.4555 0.887 0.0759>}
 dyn_hair_01_lf {parent=base_hair_01_lf, loc=<0 0 0>, rot=<0 0 0 1>}
 base_hair_02_lf {parent=head, loc=<6.3834 0.405 -7.9615>, rot=<-0.0345 0.6398 0.7634 0.0821>}
 dyn_hair_02_lf {parent=base_hair_02_lf, loc=<0 0 0>, rot=<0 0 0 1>}
 base_hair_01_rt {parent=head, loc=<4.6567 7.02 6.1424>, rot=<-0.1077 -0.3213 0.9387 0.063>}
 dyn_hair_01_rt {parent=base_hair_01_rt, loc=<0 0 0>, rot=<0 0 0 1>}
 base_hair_02_rt {parent=head, loc=<4.5491 4.6261 7.0187>, rot=<-0.0383 -0.3441 0.9373 0.0391>}
 dyn_hair_02_rt {parent=base_hair_02_rt, loc=<0 0 0>, rot=<0 0 0 1>}
 base_hair_03_rt {parent=head, loc=<4.0322 0.3647 8.2798>, rot=<-0.0086 -0.6539 0.7564 -0.012>}
 dyn_hair_03_rt {parent=base_hair_03_rt, loc=<0 0 0>, rot=<0 0 0 1>}
 base_hair_back_01_cr {parent=head, loc=<4.0909 -4.1075 0>, rot=<0.2808 -0.9598 0 0>}
 dyn_hair_back_01_cr {parent=base_hair_back_01_cr, loc=<0 0 0>, rot=<0 0 0 1>}
 dyn_hair_back_02_cr {parent=dyn_hair_back_01_cr, loc=<6.253 0 0>, rot=<0 0 0 1>}
 base_tail_hair_01_lf {parent=dyn_hair_back_02_cr, loc=<4.9352 1.3525 1.128>, rot=<0.4431 0.4097 0.1646 -0.7802>}
 dyn_tail_hair_01_lf {parent=base_tail_hair_01_lf, loc=<0 0 0>, rot=<0 0 0 1>}
 dyn_tail_hair_02_lf {parent=dyn_tail_hair_01_lf, loc=<1.2094 0 0>, rot=<-0.2139 0.2281 0.1368 -0.9399>}
 base_tail_hair_03_lf {parent=dyn_tail_hair_02_lf, loc=<3.4846 0 0>, rot=<0.2309 0.5427 0.0395 -0.8066>}
 dyn_tail_hair_03_lf {parent=base_tail_hair_03_lf, loc=<0 0 0>, rot=<0 0 0 1>}
 dyn_tail_hair_04_lf {parent=dyn_tail_hair_03_lf, loc=<4.5 0 0>, rot=<0.0006 0.1668 -0.2184 -0.9615>}
 dyn_tail_hair_05_lf {parent=dyn_tail_hair_04_lf, loc=<4.5 0 0>, rot=<0.0067 0.1876 -0.3162 -0.9299>}
 dyn_tail_hair_06_lf {parent=dyn_tail_hair_05_lf, loc=<4.5 0 0>, rot=<-0.0116 -0.0453 -0.0994 -0.9939>}
 dyn_tail_hair_07_lf {parent=dyn_tail_hair_06_lf, loc=<4.5 0 0>, rot=<-0.0001 -0.0211 -0.0133 -0.9997>}
 dyn_tail_hair_ik_07_lf {parent=dyn_tail_hair_07_lf, loc=<0 0 0>, rot=<0 0 0 1>}
 dyn_tail_hair_ik_06_lf {parent=dyn_tail_hair_06_lf, loc=<0 0 0>, rot=<0 0 0 1>}
 dyn_tail_hair_ik_05_lf {parent=dyn_tail_hair_05_lf, loc=<0 0 0>, rot=<0 0 0 1>}
 dyn_tail_hair_ik_04_lf {parent=dyn_tail_hair_04_lf, loc=<0 0 0>, rot=<0 0 0 1>}
 dyn_tail_hair_ik_03_lf {parent=dyn_tail_hair_03_lf, loc=<0 0 0>, rot=<0 0 0 1>}
 base_necklace_01_cr {parent=neck_01, loc=<0 0 0>, rot=<0.3884 0.5909 0.5909 0.3884>}
 dyn_necklace_01_cr {parent=base_necklace_01_cr, loc=<0 0 0>, rot=<0 0 0 1>}
 dyn_necklace_02_cr {parent=dyn_necklace_01_cr, loc=<4.2279 0 0>, rot=<0 0 0 1>}
 dyn_necklace_03_cr {parent=dyn_necklace_02_cr, loc=<4.2279 0 0>, rot=<0 0 0 1>}
 dyn_necklace_04_cr {parent=dyn_necklace_03_cr, loc=<4.2279 0 0>, rot=<0 0 0 1>}
 clavicle_l {parent=spine_05, loc=<5.4831 -0.1053 -3.2946>, rot=<0.0465 -0.7289 -0.009 -0.6829>}
 upperarm_l {parent=clavicle_l, loc=<9.8658 -1.2136 -0.4105>, rot=<0.0164 -0.3245 -0.0269 -0.9454>}
 upperarm_twist_01_l {parent=upperarm_l, loc=<7.2135 -0.1134 0>, rot=<0 0 0 1>}
 null1 {parent=upperarm_twist_01_l, loc=<0 0 0>, rot=<0 0 0 1>}
 deform_upperarm_fr_base_l {parent=upperarm_twist_01_l, loc=<-5.6312 5.092 0.3354>, rot=<0 0 0 -1>}
 deform_upperarm_fr_l {parent=deform_upperarm_fr_base_l, loc=<0 0 0>, rot=<0 0 0 1>}
 upperarm_twist_02_l {parent=upperarm_l, loc=<19.7998 -0.1134 0>, rot=<0 0 0 1>}
 dyn_tail_hair_for_aim {parent=clavicle_l, loc=<5.3056 3.4484 10.8631>, rot=<0.3511 0.3921 -0.4359 -0.73>}
 clavicle_r {parent=spine_05, loc=<5.4831 -0.1053 3.2946>, rot=<0.7289 0.0465 -0.6829 0.009>}
 upperarm_r {parent=clavicle_r, loc=<-9.8658 1.2136 0.4105>, rot=<0.0164 -0.3245 -0.0269 -0.9454>}
 upperarm_twist_01_r {parent=upperarm_r, loc=<-7.2135 0.1134 0>, rot=<0 0 0 -1>}
 deform_upperarm_fr_base_r {parent=upperarm_twist_01_r, loc=<5.6312 -5.092 -0.3354>, rot=<0 0 0 1>}
 deform_upperarm_fr_r {parent=deform_upperarm_fr_base_r, loc=<0 0 0>, rot=<0 0 0 1>}
 upperarm_twist_02_r {parent=upperarm_r, loc=<-19.7998 0.1134 0>, rot=<0 0 0 -1>}
 base_neck_shirt_lf {parent=spine_05, loc=<9.5111 4.5144 -7.7962>, rot=<0.1717 -0.6859 0.1162 -0.6975>}
 dyn_neck_shirt_lf {parent=base_neck_shirt_lf, loc=<0 0 0>, rot=<0 0 0 1>}
 aim_necklace_01_cr {parent=spine_05, loc=<-5.0127 15.664 0>, rot=<0.2622 0.6567 0.6567 0.2622>}
 base_neck_shirt_rt {parent=spine_05, loc=<9.5108 4.5143 7.7962>, rot=<-0.6859 -0.1717 0.6975 0.1162>}
 dyn_neck_shirt_rt {parent=base_neck_shirt_rt, loc=<0 0 0>, rot=<0 0 0 1>}
 dyn_hair_back_aim {parent=spine_05, loc=<6.9978 -13.7159 0>, rot=<0.3151 -0.949 0 0>}
 aiming up pistol {type=Curve, loc=<0 0 0>, rot=<0 0 0 1>}
]
variables [
 RBAN enabled {type=bool, default=False}
 aiming pistol {type=double, default=0}
 clavicle up for targeting {type=float, default=0}
 crouching move {type=double, default=0}
 head crouching {type=double, default=0}
 is Crouching {type=bool, default=False}
 is Targeting {type=bool, default=False}
]
