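# Physics asset for the HangNine female medium character's FaceAcc part, set up around a left-side hair tail.
# Constraints chain capsules on dyn_tail_hair_03_lf through 07_lf from base_tail_hair_03_lf, with tight swing and twist limit angles.
# Capsules on head, neck, upper spine, left clavicle, left upper arm and left neck-shirt bone approximate the surrounding body.
# Several base and upper hair bones, including the chain's anchor, carry zero-size capsules.
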PhysicsAsset /BRCosmetics/Characters/Player/Female/Medium/Bodies/F_MED_HangNine/Meshes/Parts/F_MED_HangNine_FaceAcc_Physics
bodies {
 base_hair_back_01_cr [{shapeType=Capsule, length=0, radius=0, shapeName=base_hair_back_01_cr_capsule, center=<5.326 0.6864 -0.7883>}]
 base_tail_hair_01_lf [{shapeType=Capsule, length=0, radius=0, shapeName=base_tail_hair_01_lf_capsule}]
 base_tail_hair_03_lf [{shapeType=Capsule, length=0, radius=0, shapeName=base_tail_hair_03_lf_capsule}]
 clavicle_l [{shapeType=Capsule, length=15.5322, radius=5.25, shapeName=clavicle_l_capsule, center=<-1.2258 -2.9932 -0.877>, rotation=<rot 90 0 0>}]
 dyn_neck_shirt_lf [{
   shapeType=Capsule
   length=6.7506
   radius=1.7654
   shapeName=upperarm_twist_01_l_capsule
   center=<4.6066 0.3863 -0.5855>
   rotation=<rot 71.3507 82.0833 -115.0486>
  }]
 dyn_tail_hair_01_lf [{shapeType=Capsule, length=0, radius=0, shapeName=dyn_tail_hair_01_lf_capsule}]
 dyn_tail_hair_02_lf [{shapeType=Capsule, length=0, radius=0, shapeName=dyn_tail_hair_02_lf_capsule}]
 dyn_tail_hair_03_lf [{
   shapeType=Capsule
   length=7.4433
   radius=2.5991
   shapeName=dyn_tail_hair_03_lf_capsule
   center=<0.8022 0.9969 0>
   rotation=<rot 90 -10.2068 0>
  }]
 dyn_tail_hair_04_lf [{shapeType=Capsule, length=4.6185, radius=1.5261, shapeName=dyn_tail_hair_04_lf_capsule, center=<2 0.9673 0>, rotation=<rot 90 0 0>}]
 dyn_tail_hair_05_lf [
  {shapeType=Capsule, length=5.9989, radius=1.5011, shapeName=dyn_tail_hair_05_lf_capsule, center=<2 0 0>, rotation=<rot 90 4.9216 0>}
  {
   shapeType=Capsule
   length=3.3996
   radius=1.0804
   shapeName=dyn_tail_hair_05_lf_capsule
   center=<4.9701 0.6962 1.8565>
   rotation=<rot 90 -10.3688 0>
  }
 ]
 dyn_tail_hair_06_lf [{shapeType=Capsule, length=6, radius=1.5, shapeName=dyn_tail_hair_06_lf_capsule, center=<1.9933 0.1569 0>, rotation=<rot 90 -1.5735 0>}]
 dyn_tail_hair_07_lf [{shapeType=Capsule, length=10.6604, radius=1.5, shapeName=dyn_tail_hair_07_lf_capsule, center=<2 0 0>, rotation=<rot 90 -4.1551 0>}]
 head [{shapeType=Capsule, length=10.1024, radius=6.8701, shapeName=head_capsule, center=<3.48 2.2834 0>, rotation=<rot 90 -24.826 0>}]
 neck_01 [{shapeType=Capsule, length=10.5659, radius=4.2387, shapeName=neck_01_capsule, center=<3.2261 0 0>, rotation=<rot 90 0 0>}]
 spine_05 [{shapeType=Capsule, length=21.0994, radius=8.6416, shapeName=spine_05_capsule, center=<0 -1.6458 0>}]
 upperarm_l [{
   shapeType=Capsule
   length=14.9417
   radius=4.2764
   shapeName=upperarm_l_capsule
   center=<7.7251 0 -0.7162>
   rotation=<rot 78.7012 -0.0003 -0.0004>
  }]
}
constraints [
 {bone1Name=dyn_tail_hair_05_lf, bone2Name=dyn_tail_hair_04_lf, swing1LimitDegrees=5, swing2LimitDegrees=0, twistLimitDegrees=0}
 {bone1Name=dyn_tail_hair_06_lf, bone2Name=dyn_tail_hair_05_lf, swing1LimitDegrees=5, swing2LimitDegrees=0, twistLimitDegrees=0}
 {bone1Name=dyn_tail_hair_07_lf, bone2Name=dyn_tail_hair_06_lf, swing1LimitDegrees=5, swing2LimitDegrees=0, twistLimitDegrees=0}
 {bone1Name=dyn_tail_hair_03_lf, bone2Name=base_tail_hair_03_lf, swing1LimitDegrees=5, swing2LimitDegrees=0, twistLimitDegrees=0}
 {bone1Name=dyn_tail_hair_04_lf, bone2Name=dyn_tail_hair_03_lf, swing1LimitDegrees=5, swing2LimitDegrees=0, twistLimitDegrees=0}
]
