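# Post-process-style AnimBP for the HangNine female hat/face accessory mesh, adding secondary motion to its hair.
# Copies parent pose, layering listed left-arm, hair-aim and collar branches from the body mesh, then runs a Control Rig fed crouch/targeting states.
# Additive bone offsets surround chained AnimDynamics and a spine_05-space RigidBody simulation.
# Events update gravity override, inverted pistol-aim curve and a zipline dynamics-scale vector, and cache the body mesh.
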AnimBlueprint /BRCosmetics/Characters/Player/Female/Medium/Bodies/F_MED_HangNine/Meshes/Parts/F_MED_HangNineHat_FaceAcc_AnimBP
target_skeleton=&/F_MED_HangNine_FaceAcc_Skeleton
graphs [
 AnimGraph {nodes [
   Output Pose(Result=Use cached pose 'dynamicsCache'()) {layerGroup=None}
   n1 = "Transform (Modify) Bone - Bone: dyn_tail_hair_02_lf"!(
    ComponentPose=Local To Component(LocalPose=Use cached pose 'controlRigCache'())
    bAlphaBoolEnabled:bool=True
    Translation="4.036,-0.593,-0.545"
    Rotation="12.1,-12.396,2"
    Scale="1,1,1"
   ) -> (Pose) {
    alphaInputType=Bool
    rotationMode=BMM_Additive
    rotationSpace=BCS_BoneSpace
    translationMode=BMM_Additive
    translationSpace=BCS_BoneSpace
    alphaBoolBlend {blendInTime=1, blendOutTime=1}
    boneToModify {boneName=dyn_tail_hair_02_lf}
    rotation=<rot 12.1 -12.396 2>
    translation=<4.036 -0.593 -0.545>
   }
   n2 = "Transform (Modify) Bone - Bone: base_hair_back_01_cr"!(ComponentPose=@n1, bAlphaBoolEnabled:bool=True, Translation="-2,0,0", Rotation="30,-10,0", Scale="1,1,1") -> (Pose) {
    alphaInputType=Bool
    rotationMode=BMM_Additive
    rotationSpace=BCS_BoneSpace
    scaleMode=BMM_Additive
    scaleSpace=BCS_BoneSpace
    translationMode=BMM_Additive
    translationSpace=BCS_BoneSpace
    alphaBoolBlend {blendInTime=1, blendOutTime=1}
    boneToModify {boneName=base_hair_back_01_cr}
    rotation {pitch=30, yaw=-10}
    translation {x=-2}
   }
   n3 = "Transform (Modify) Bone - Bone: base_tail_hair_03_lf"!(ComponentPose=@n2, bAlphaBoolEnabled:bool=True, Translation="0,1,0", Rotation="10,0,0", Scale="1,1,1") -> (Pose) {
    alphaInputType=Bool
    rotationMode=BMM_Additive
    rotationSpace=BCS_BoneSpace
    scaleMode=BMM_Additive
    scaleSpace=BCS_BoneSpace
    translationMode=BMM_Additive
    translationSpace=BCS_BoneSpace
    alphaBoolBlend {blendInTime=1, blendOutTime=1}
    boneToModify {boneName=base_tail_hair_03_lf}
    rotation {pitch=10}
    translation {y=1}
   }
   n4 = "Transform (Modify) Bone - Bone: dyn_hair_back_02_cr"!(ComponentPose=@n3, bAlphaBoolEnabled:bool=True, Translation="0,0,0", Rotation="0,0,-20", Scale="1,1,1") -> (Pose) {
    alphaInputType=Bool
    rotationMode=BMM_Additive
    rotationSpace=BCS_BoneSpace
    scaleMode=BMM_Additive
    scaleSpace=BCS_BoneSpace
    translationMode=BMM_Additive
    translationSpace=BCS_BoneSpace
    alphaBoolBlend {blendInTime=1, blendOutTime=1}
    boneToModify {boneName=dyn_hair_back_02_cr}
    rotation {roll=-20}
   }
   Save cached pose 'modifyTransCache'(Pose=Component To Local(ComponentPose=@n4)) {cacheName=modifyTransCache}
   n5 = Anim Dynamics - Bone: dyn_hair_01_cr!(
    ComponentPose=Local To Component(LocalPose=Use cached pose 'modifyTransCache'())
    Alpha=1
    ComponentLinearAccScale:Vector="0.5,0.5,0.5"
    ComponentLinearVelScale:Vector="0.5,0.5,0.5"
   ) -> (Pose) {
    angularDampingOverride=1
    angularSpringConstant=20000
    bAngularSpring=true
    bGravityOverrideInSimSpace=true
    bOverrideAngularDamping=true
    bUseGravityOverride=true
    gravityScale=0.35
    simulationSpace=RootRelative
    boundBone {boneName=dyn_hair_01_cr}
    componentAppliedLinearAccClamp=<3000 3000 3000>
    componentLinearAccScale=<0.5 0.5 0.5>
    componentLinearVelScale=<0.5 0.5 0.5>
    gravityOverride {z=-980}
    physicsBodyDefinitions [{
      boundBone {boneName=dyn_hair_01_cr}
      boxExtents=<30 30 30>
      constraintSetup {angularLimitsMax=<0 15 0>, angularLimitsMin=<0 -15 -10>, angularTarget=<1 0 0>}
      localJointOffset=<8 0 0>
     }]
    simSpaceSettings {maxAngularAcceleration=35, maxAngularVelocity=7, simSpaceAngularAlpha=1}
   }
   n6 = Anim Dynamics - Bone: dyn_hair_01_lf!(ComponentPose=@n5, Alpha=1, ComponentLinearAccScale:Vector="0.5,0.5,0.5", ComponentLinearVelScale:Vector="0.5,0.5,0.5") -> (Pose) {
    angularDampingOverride=1
    angularSpringConstant=20000
    bAngularSpring=true
    bGravityOverrideInSimSpace=true
    bOverrideAngularDamping=true
    bUseGravityOverride=true
    gravityScale=0.35
    simulationSpace=RootRelative
    boundBone {boneName=dyn_hair_01_lf}
    componentAppliedLinearAccClamp=<3000 3000 3000>
    componentLinearAccScale=<0.5 0.5 0.5>
    componentLinearVelScale=<0.5 0.5 0.5>
    gravityOverride {z=-980}
    physicsBodyDefinitions [{
      boundBone {boneName=dyn_hair_01_lf}
      boxExtents=<30 30 30>
      constraintSetup {angularLimitsMax=<0 5 0>, angularLimitsMin=<0 -5 -10>, angularTarget=<1 0 0>}
      localJointOffset=<8 0 0>
     }]
    simSpaceSettings {maxAngularAcceleration=35, maxAngularVelocity=7, simSpaceAngularAlpha=1}
   }
   n7 = Anim Dynamics - Bone: dyn_hair_02_lf!(ComponentPose=@n6, Alpha=1, ComponentLinearAccScale:Vector="0.5,0.5,0.5", ComponentLinearVelScale:Vector="0.5,0.5,0.5") -> (Pose) {
    angularDampingOverride=1
    angularSpringConstant=20000
    bAngularSpring=true
    bGravityOverrideInSimSpace=true
    bOverrideAngularDamping=true
    bUseGravityOverride=true
    gravityScale=0.35
    simulationSpace=RootRelative
    boundBone {boneName=dyn_hair_02_lf}
    componentAppliedLinearAccClamp=<3000 3000 3000>
    componentLinearAccScale=<0.5 0.5 0.5>
    componentLinearVelScale=<0.5 0.5 0.5>
    gravityOverride {z=-980}
    physicsBodyDefinitions [{
      boundBone {boneName=dyn_hair_02_lf}
      boxExtents=<30 30 30>
      constraintSetup {angularLimitsMax=<0 10 0>, angularLimitsMin=<0 -10 -15>, angularTarget=<1 0 0>}
      localJointOffset=<8 0 0>
     }]
    simSpaceSettings {maxAngularAcceleration=35, maxAngularVelocity=7, simSpaceAngularAlpha=1}
   }
   n8 = Anim Dynamics - Bone: dyn_hair_01_rt!(ComponentPose=@n7, Alpha=1, ComponentLinearAccScale:Vector="0.5,0.5,0.5", ComponentLinearVelScale:Vector="0.5,0.5,0.5") -> (Pose) {
    angularDampingOverride=1
    angularSpringConstant=20000
    bAngularSpring=true
    bGravityOverrideInSimSpace=true
    bOverrideAngularDamping=true
    bUseGravityOverride=true
    gravityScale=0.35
    simulationSpace=RootRelative
    boundBone {boneName=dyn_hair_01_rt}
    componentAppliedLinearAccClamp=<3000 3000 3000>
    componentLinearAccScale=<0.5 0.5 0.5>
    componentLinearVelScale=<0.5 0.5 0.5>
    gravityOverride {z=-980}
    physicsBodyDefinitions [{
      boundBone {boneName=dyn_hair_01_rt}
      boxExtents=<30 30 30>
      constraintSetup {angularLimitsMax=<0 30 0>, angularLimitsMin=<0 -30 -10>, angularTarget=<1 0 0>}
      localJointOffset=<4 0 0>
     }]
    simSpaceSettings {maxAngularAcceleration=35, maxAngularVelocity=7, simSpaceAngularAlpha=1}
   }
   n9 = Anim Dynamics - Bone: dyn_hair_02_rt!(ComponentPose=@n8, Alpha=1, ComponentLinearAccScale:Vector="0.5,0.5,0.5", ComponentLinearVelScale:Vector="0.5,0.5,0.5") -> (Pose) {
    angularDampingOverride=1
    angularSpringConstant=20000
    bAngularSpring=true
    bGravityOverrideInSimSpace=true
    bOverrideAngularDamping=true
    bUseGravityOverride=true
    gravityScale=0.35
    simulationSpace=RootRelative
    boundBone {boneName=dyn_hair_02_rt}
    componentAppliedLinearAccClamp=<3000 3000 3000>
    componentLinearAccScale=<0.5 0.5 0.5>
    componentLinearVelScale=<0.5 0.5 0.5>
    gravityOverride {z=-980}
    physicsBodyDefinitions [{
      boundBone {boneName=dyn_hair_02_rt}
      boxExtents=<30 30 30>
      constraintSetup {angularLimitsMax=<0 8 0>, angularLimitsMin=<0 -10 -15>, angularTarget=<1 0 0>}
      localJointOffset=<10 0 0>
     }]
    simSpaceSettings {maxAngularAcceleration=35, maxAngularVelocity=7, simSpaceAngularAlpha=1}
   }
   n10 = Anim Dynamics - Bone: dyn_hair_03_rt!(ComponentPose=@n9, Alpha=1, ComponentLinearAccScale:Vector="0.5,0.5,0.5", ComponentLinearVelScale:Vector="0.5,0.5,0.5") -> (Pose) {
    angularDampingOverride=1
    angularSpringConstant=20000
    bAngularSpring=true
    bGravityOverrideInSimSpace=true
    bOverrideAngularDamping=true
    bUseGravityOverride=true
    gravityScale=0.35
    simulationSpace=RootRelative
    boundBone {boneName=dyn_hair_03_rt}
    componentAppliedLinearAccClamp=<3000 3000 3000>
    componentLinearAccScale=<0.5 0.5 0.5>
    componentLinearVelScale=<0.5 0.5 0.5>
    gravityOverride {z=-980}
    physicsBodyDefinitions [{
      boundBone {boneName=dyn_hair_03_rt}
      boxExtents=<30 30 30>
      constraintSetup {angularLimitsMax=<0 20 0>, angularLimitsMin=<0 -20 -15>, angularTarget=<1 0 0>}
      localJointOffset=<4 0 0>
     }]
    simSpaceSettings {maxAngularAcceleration=35, maxAngularVelocity=7, simSpaceAngularAlpha=1}
   }
   n11 = RigidBody!(ComponentPose=@n10, Alpha, OverrideWorldGravity:Vector="0,0,0") -> (Pose) {
    bClampLinearTranslationLimitToRefPose=true
    bOverrideWorldGravity=true
    overridePhysicsAsset=&/F_MED_HangNine_FaceAcc_Physics
    simulationSpace=BaseBoneSpace
    baseBoneRef {boneName=spine_05}
   }
   n12 = "Transform (Modify) Bone - Bone: dyn_tail_hair_04_lf"(ComponentPose=@n11, Alpha=1, Translation="0,0,0", Rotation="0,15,-5", Scale="1,1,1") {
    rotationMode=BMM_Additive
    rotationSpace=BCS_BoneSpace
    alphaScaleBiasClamp {bInterpResult=true}
    boneToModify {boneName=dyn_tail_hair_04_lf}
    rotation {roll=-5, yaw=15}
   }
   n13 = "Transform (Modify) Bone - Bone: dyn_tail_hair_06_lf"(ComponentPose=@n12, Alpha=1, Translation="0,0,0", Rotation="20.536,-1.488,-2.449", Scale="1,1,1") {
    rotationMode=BMM_Additive
    rotationSpace=BCS_BoneSpace
    alphaScaleBiasClamp {bInterpResult=true}
    boneToModify {boneName=dyn_tail_hair_06_lf}
    rotation=<rot 20.536 -1.488 -2.449>
   }
   n14 = "Transform (Modify) Bone - Bone: dyn_tail_hair_07_lf"(ComponentPose=@n13, Alpha=1, Translation="0,0,0", Rotation="23.061,15.652,-7.218", Scale="1,1,1") {
    rotationMode=BMM_Additive
    rotationSpace=BCS_BoneSpace
    alphaScaleBiasClamp {bInterpResult=true}
    boneToModify {boneName=dyn_tail_hair_07_lf}
    rotation=<rot 23.061 15.652 -7.218>
   }
   n15 = "Transform (Modify) Bone - Bone: dyn_tail_hair_04_lf"!(ComponentPose=@n14, bAlphaBoolEnabled:bool=True, Translation="0,1,0", Rotation="-5.795,10,2.65", Scale="1,1,1") -> (Pose) {
    alphaInputType=Bool
    rotationMode=BMM_Additive
    rotationSpace=BCS_BoneSpace
    scaleMode=BMM_Additive
    scaleSpace=BCS_BoneSpace
    translationMode=BMM_Additive
    translationSpace=BCS_BoneSpace
    boneToModify {boneName=dyn_tail_hair_04_lf}
    rotation=<rot -5.795 10 2.65>
    translation {y=1}
   }
   Save cached pose 'dynamicsCache'(Pose=Component To Local(ComponentPose=@n15)) {cacheName=dynamicsCache}
   n16 = Control Rig!(
    Source=Use cached pose 'StartPose'()
    Alpha
    is Crouching:bool=Get bIsCrouching()
    is Targeting:bool=Get bIsTargeting()
    RBAN enabled:bool=Get bRigidBodyEnabled()
   ) -> (Pose) {controlRigAssetReference {blueprintRigClass=&/F_MED_HangNine_FaceAcc_CtrlRig.F_MED_HangNine_FaceAcc_CtrlRig_C}}
   Save cached pose 'controlRigCache'(Pose=@n16) {cacheName=controlRigCache}
   n17 = Layered blend per bone(
    BasePose=Copy Pose From Mesh() {bCopyCurves=true, bUseAttachedParent=true}
    BlendPoses_0=Copy Pose From Mesh(SourceMeshComponent=Get Body_Mesh())
   ) {layerSetup [{branchFilters [
       {boneName=clavicle_l}
       {boneName=upperarm_l}
       {boneName=upperarm_twist_01_l}
       {boneName=upperarm_twist_02_l}
       {boneName=dyn_tail_hair_for_aim}
       {boneName=dyn_hair_back_aim}
       {boneName=base_neck_shirt_lf}
       {boneName=dyn_neck_shirt_lf}
       {boneName=deform_upperarm_fr_l}
      ]}]}
   Save cached pose 'StartPose'(Pose=@n17) {cacheName=StartPose}
  ]}
 EventGraph {nodes [
   Event BlueprintUpdateAnimation()
   > n1 = "Gravity Override State Function (Parameter Asset) (Nativized)"(
    CharacterPartAnimInstance=Self-Reference()
    GameTimeInSeconds=GetGameTimeInSeconds()
    Parameters=&/F_MED_HangNine_Parameters
   )
   > Set GravityOverride(GravityOverride=@n1)
   > Set aiming up pistol(aiming up pistol:double=float - float(A=1, B=GetCurveValue(CurveName=aiming up pistol))) -> (Output_Get:double)
   n2 = Event BlueprintProcessAnimInputEvent()
   n3 = IsNamed(self=@n2.AnimInputEvent, InEventName=aie.Player.Zipline.Entered)
   n4 = OR Boolean(A=Get bIsMoving2D(), B=Get bIsTacticalSprinting())
   n5 = Get bIsJumping()
   > n6 = Branch(Condition=OR Boolean(A=@n3, B=AND Boolean!(A=@n3, B=@n5, C:bool=@n4) -> (ReturnValue)))
   > Set animDyn on zipline(animDyn on zipline:Vector="0, 0, 0") -> (Output_Get:Vector)
   Branch(execute=@n6.else, Condition=IsNamed(self=@n2.AnimInputEvent, InEventName=aie.Player.Zipline.Exited))
   > Set animDyn on zipline(animDyn on zipline="0.5,0.5,0.5")
   Event OnCharacterCustomizationLoaded()
   > n7 = Cast To PlayerPawn_Athena(Object=TryGetPawnOwner())
   > Set Body_Mesh(Body_Mesh=GetSkeletalMeshForPartType(self=@n7))
  ]}
]
variables [
 aiming up pistol {type=number}
 animDyn on zipline {type=Vector, default=<0.5 0.5 0.5>}
 Body_Mesh {type=SkeletalMeshComponent}
 GravityOverride {type=Vector}
]
